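; Clyde heist music manager that drives heisting, lockdown and stinger music for POI bases.
; The authority caches base managers from the actor registry, including late registrations, to resolve the current base.
; Lockdown intro, loop and outro play quantized on a Quartz clock, with a timer that ends the loop.
; Audio gameplay volumes and pawn death fade music out, tracks register with the MusicEventSubsystem, and stingers are rate-limited by a timed gate.
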(event EventEndPlay
  (params (EndPlayReason :type EEndPlayReason%20Enum :default "Destroyed"))
  ;[0.02445 0.05521 0.02107] EndPlay & Handle Cleanup
  (Utilities|FlowControl|Sequence
    (:then_0
      ;;[1 0.4462 0] [0] Clear and invalidate all timers
      (Utilities|Time|ClearandInvalidateTimerbyHandle (Variables|Default|GetHeistMusicHandle))
      (Utilities|Time|ClearandInvalidateTimerbyHandle (Variables|Conditions|GetLockdownMusicHandle)))
    (:then_1
      ;;[1 0.4462 0] [1] Unbind all bound Events
      (bind _cacheddataregistry (Variables|BaseRegistration|GetCachedDataRegistry)
        (:then
          (EventDispatchers|UnbindEventfromOnActorRegistered _cacheddataregistry (EventDispatchers|CreateEvent :!event CacheRegisteredManager)))
        (:else
          _))
      (for _array_element (Variables|BaseRegistration|GetBaseManagers)
        (Utilities|IsValid _array_element)))
    (:then_2
      ;;[1 0.4462 0] [2] Call Parent End Play last to guarantee all custom cleanup happens
      (|CallParentFunction EndPlayReason))))

(event EventBeginPlay
  ;[0.009152 0.08762 0.1594] Begin Play
  (Networking|SwitchHasAuthority
    (:Authority
      (Class|FortAsyncActionCurrentPlaylistReady|PlaylistDataReadyAsync
        (:OnReady
          (BaseRegistration|GetandRegisterBaseManagers))))))

(event Custom|StartLockdownTrack
  (params (HeistBaseID :type Gameplay%20Tag%20Structure))
  ;[0.2302 0.007375 0.002337 1 25] StartLockdownMusic
  (Variables|Conditions|SetCurrentHeistState "NewEnumerator2")
  (Variables|BaseRegistration|GetCurrentBaseManager
    (:then
      (Utilities|FlowControl|Delay (Variables|Conditions|GetDelayBeforeLockdownMusic))
      (CallFunction|PlayLockdownMusic HeistBaseID)
      (bind _returnvalue (Utilities|Time|SetTimerbyFunctionName 0 "StopLockdownTrackIfStillInPOI" 1.8e+02))
      (Variables|Conditions|SetLockdownMusicHandle _returnvalue))
    (:else)))

(event OnListenerEnterEvent(AGV_GrottoPOI))

(event Custom|LockdownStatusChange
  (params (AlertedActor :type Actor%20Object%20Reference) (HeistStatus :type Enum_HeistAudioStatus%20Enum :default "NewEnumerator3") (POI_Tag :type Gameplay%20Tag%20Structure)))

(event Custom|StopLockdownTrackIfStillInPOI
  ;[0.4302 0.07371 0.1365] If Lockdown Time Is Over, Stop Lockdown Music If Player Is Still Within The POI
  (bind _lockdowntrack (Variables|Default|GetLockdownTrack)
    (:then
      (if (Audio|Components|Audio|IsPlaying _lockdowntrack)
        (if (Variables|Conditions|GetbInsideCurrentPOI)
          (CallFunction|PlayLockdownOutro)
          (QuartzClock|NotifyonQuantizationBoundary (Variables|Quartz|GetLockdownClock) (Variables|Quartz|GetOutroMusicBoundary) (EventDispatchers|CreateEvent :!event Custom|StopMainTrack) 0.0))))
    (:else)))

(event OnListenerEnterEvent(AGV_YachtPOI))

(event OnListenerEnterEvent(AGV_CattyPOI))

(event Custom|StopMainTrack
  (params (EventType :type EQuartzCommandDelegateSubType%20Enum :default "CommandOnFailedToQueue") (Name :type Name))
  (bind _lockdowntrack (Variables|Default|GetLockdownTrack))
  ;[0.4302 0.07371 0.1365] Stop Lockdown Loop and play "outro" section
  (switch Utilities|FlowControl|Switch|SwitchonEQuartzCommandDelegateSubType EventType
    (:CommandOnStarted
      (Audio|RemoveEvent (GameInstanceSubsystems|GetMusicEventSubsystem) _lockdowntrack)
      (if (not (Variables|Conditions|GetbIsFadingOut))
        (Audio|Components|Audio|Stop _lockdowntrack)
        (Utilities|Time|ClearandInvalidateTimerbyHandle (Variables|Conditions|GetLockdownMusicHandle))
        (Utilities|FlowControl|Delay (Variables|Default|GetDelayTimeToManuallyResetHeistState))
        (Variables|Conditions|SetCurrentHeistState "NewEnumerator3")
        (QuartzClockHandle|DeleteClockbyHandle (WorldSubsystems|GetQuartzSubsystem) (Variables|Quartz|GetLockdownClock))))))

(event OnListenerEnterEvent(AGV_SharkPOI))

(event OnListenerEnterEvent(AGV_AgencyPOI))

(event OnListenerEnterEvent(AGV_OilRigPOI))

(event Custom|ReplicatedTrySetLateBaseManager
  (params (HeistBaseTagID :type Gameplay%20Tag%20Structure))
  ;[0 0 0] ReplicatedTrySetLateBaseManager
  ; Fallback implementation in PIE for when a player could enter a AGV before a base is registered on the Server
  (Networking|CanExecuteCosmeticEvents
    (:True
      (if (Variables|BaseRegistration|GetbShouldFindBaseOnRegistration)
        (Default|SetCurrentBaseManager HeistBaseTagID) ; will only set if the newly registered base matches the existing tag
      ))) ; Only on clients
)

(event Custom|FadeMusicOutWhenPlayerDies
  (params (DestroyedActor :type Actor%20Object%20Reference))
  ;[0.009152 0.08762 0.1594] Handle Music When Enter/Exiting Audio Gameplay Volumes
  ;;[0 0.2703 1] Player Died
  (CallFunction|FadeOutHeistingMusic)
  (CallFunction|FadeOutLockdownMusic))

(event OnListenerEnterEvent(AGV_Shark_HeistMusicOff)
  ;[0.009152 0.08762 0.1594] Handle Music When Enter/Exiting Audio Gameplay Volumes
  ;;[0 0.2703 1] Stop Heisting Music
  (if (Utilities|Enum|Equal(Enum) (Variables|Conditions|GetCurrentHeistState) "NewEnumerator2")
    (CallFunction|FadeOutLockdownMusic)
    (else
      (CallFunction|FadeOutHeistingMusic))))

(event OnListenerEnterEvent(AGV_Agency_HeistMusicOff)
  ;[0.009152 0.08762 0.1594] Handle Music When Enter/Exiting Audio Gameplay Volumes
  ;;[0 0.2703 1] Stop Heisting Music
  (if (Utilities|Enum|Equal(Enum) (Variables|Conditions|GetCurrentHeistState) "NewEnumerator2")
    (CallFunction|FadeOutLockdownMusic)
    (else
      (CallFunction|FadeOutHeistingMusic))))

(event OnListenerEnterEvent(AGV_Grotto_HeistMusicOff)
  ;[0.009152 0.08762 0.1594] Handle Music When Enter/Exiting Audio Gameplay Volumes
  ;;[0 0.2703 1] Stop Heisting Music
  (if (Utilities|Enum|Equal(Enum) (Variables|Conditions|GetCurrentHeistState) "NewEnumerator2")
    (CallFunction|FadeOutLockdownMusic)
    (else
      (CallFunction|FadeOutHeistingMusic))))

(event OnListenerEnterEvent(AGV_Grotto_HeistMusicOff2)
  ;[0.009152 0.08762 0.1594] Handle Music When Enter/Exiting Audio Gameplay Volumes
  ;;[0 0.2703 1] Stop Heisting Music
  (if (Utilities|Enum|Equal(Enum) (Variables|Conditions|GetCurrentHeistState) "NewEnumerator2")
    (CallFunction|FadeOutLockdownMusic)
    (else
      (CallFunction|FadeOutHeistingMusic))))

(event OnListenerExitEvent(AGV_Resort_LckDownMx)
  ;[0.009152 0.08762 0.1594] Handle Music When Enter/Exiting Audio Gameplay Volumes
  ;;[0 0.2703 1] Stop Lockdown Music
  (Variables|Conditions|SetbInsideCurrentPOI false)
  (CallFunction|FadeOutLockdownMusic)
  (CallFunction|FadeOutHeistingMusic))

(event OnListenerEnterEvent(AGV_Grotto_HeistMusicOff3)
  ;[0.009152 0.08762 0.1594] Handle Music When Enter/Exiting Audio Gameplay Volumes
  ;;[0 0.2703 1] Stop Heisting Music
  (if (Utilities|Enum|Equal(Enum) (Variables|Conditions|GetCurrentHeistState) "NewEnumerator2")
    (CallFunction|FadeOutLockdownMusic)
    (else
      (CallFunction|FadeOutHeistingMusic))))

(event OnListenerExitEvent(AGV_EstatePOI_LckDownMx)
  ;[0.009152 0.08762 0.1594] Handle Music When Enter/Exiting Audio Gameplay Volumes
  ;;[0 0.2703 1] Stop Lockdown Music
  (Variables|Conditions|SetbInsideCurrentPOI false)
  (CallFunction|FadeOutLockdownMusic)
  (CallFunction|FadeOutHeistingMusic))

(event OnListenerExitEvent(AGV_HotelPOI_LckDownMx)
  ;[0.009152 0.08762 0.1594] Handle Music When Enter/Exiting Audio Gameplay Volumes
  ;;[0 0.2703 1] Stop Lockdown Music
  (Variables|Conditions|SetbInsideCurrentPOI false)
  (CallFunction|FadeOutLockdownMusic)
  (CallFunction|FadeOutHeistingMusic))

(event OnListenerEnterEvent(AGV_OilRig_HeistMusicOff)
  ;[0.009152 0.08762 0.1594] Handle Music When Enter/Exiting Audio Gameplay Volumes
  ;;[0 0.2703 1] Stop Heisting Music
  (if (Utilities|Enum|Equal(Enum) (Variables|Conditions|GetCurrentHeistState) "NewEnumerator2")
    (CallFunction|FadeOutLockdownMusic)
    (else
      (CallFunction|FadeOutHeistingMusic))))

(event OnListenerEnterEvent(AGV_Catty_HeistMusicOff)
  ;[0.009152 0.08762 0.1594] Handle Music When Enter/Exiting Audio Gameplay Volumes
  ;;[0 0.2703 1] Stop Heisting Music
  (if (Utilities|Enum|Equal(Enum) (Variables|Conditions|GetCurrentHeistState) "NewEnumerator2")
    (CallFunction|FadeOutLockdownMusic)
    (else
      (CallFunction|FadeOutHeistingMusic))))

(event OnListenerEnterEvent(AGV_Yacht_HeistMusicOff)
  ;[0.009152 0.08762 0.1594] Handle Music When Enter/Exiting Audio Gameplay Volumes
  ;;[0 0.2703 1] Stop Heisting Music
  (if (Utilities|Enum|Equal(Enum) (Variables|Conditions|GetCurrentHeistState) "NewEnumerator2")
    (CallFunction|FadeOutLockdownMusic)
    (else
      (CallFunction|FadeOutHeistingMusic))))

(event OnListenerEnterEvent(AGV_Hotel_HeistMusicOff_05)
  ;[0.009152 0.08762 0.1594] Handle Music When Enter/Exiting Audio Gameplay Volumes
  ;;[0 0.2703 1] Stop Heisting Music
  (if (Utilities|Enum|Equal(Enum) (Variables|Conditions|GetCurrentHeistState) "NewEnumerator2")
    (CallFunction|FadeOutLockdownMusic)
    (else
      (CallFunction|FadeOutHeistingMusic))))

(fn AudioDebug
  (params (InString :type String))
  (if (Variables|Conditions|GetPrintDebug)
    (Development|PrintString (Utilities|String|Append "Heist Audio: " "Music Event: " InString) true true "(R=0.000000,G=0.660000,B=1.000000,A=1.000000)" 5.0)))

(fn BP_HeistMusicManager_AutoGenFunc
  (params (ClockName :type Name) (QuantizationType :type EQuartzCommandQuantization%20Enum :default "Bar") (NumBars :type Integer) (Beat :type Integer) (BeatFraction :type Float%20%28single-precision%29))
  (switch Utilities|FlowControl|Switch|SwitchonEQuartzCommandQuantization QuantizationType))

(fn CacheRegisteredManager
  (params (GameplayTag :type Gameplay%20Tag%20Structure%20%28by%20ref%29) (RegisteredActor :type Actor%20Object%20Reference))
  ;[0.4973 0.715 1] CacheRegisteredBaseDefense
  ; Add the registered heist base manager - tag was already checked by the registry.
  (bind _asbga_clyde_base_manager (Utilities|Casting|CastToBGA_ClydeBaseManager RegisteredActor))
  (Utilities|Array|AddUnique (Variables|BaseRegistration|GetBaseManagers) _asbga_clyde_base_manager)
  (Networking|FlushNetDormancy) ; Force replication of BaseManagers,\nbecause NetDormancy=DormanAll
  ;;[1 0 0.9544] Fallback implementation: In case of a late base manager registration. This is not expected outside of PIE when players can load into a base directly.
  (CallFunction|ReplicatedTrySetLateBaseManager (Variables|/Clyde/Audio/Items/BGA_ClydeBaseManager.BGA_ClydeBaseManager_C|GetHeistBaseId%20Tag%20Const _asbga_clyde_base_manager)))

(fn CheckIfPreviouslyLockedDown
  (switch int (Utilities|Array|FindItem (Variables|Tags|GetPOITagArray) (Variables|BaseRegistration|GetCurrentHeistBaseTag))
    (:0
      (if (Variables|Conditions|GetbEstateHasBeenLockedDown)
        (Variables|Conditions|SetbCurrentBaseHasBeenLockedDown true)
        (else
          (Variables|Conditions|SetbCurrentBaseHasBeenLockedDown false)))
      (return (Variables|Conditions|GetbCurrentBaseHasBeenLockedDown)))
    (:1
      (if (Variables|Conditions|GetbHotelHasBeenLockedDown)
        _
        (else
          _)))
    (:2
      (if (Variables|Conditions|GetbResortHasBeenLockedDown)
        _
        (else
          _)))))

(fn CreateLockdownClock
  (bind _returnvalue (QuartzClockHandle|CreateNewClock (WorldSubsystems|GetQuartzSubsystem) "HeistLockdownClock" (Utilities|Struct|MakeQuartzClockSettings (Utilities|Struct|MakeQuartzTimeSignature 4)) false true))
  (bind _output_get (Variables|Quartz|SetLockdownClock _returnvalue))
  (bind _clockhandle (Quantization|SetBeatsPerMinute :self _output_get :BeatsPerMinute 1.1e+02))
  (QuartzClock|SubscribetoAllQuantizationEvents _clockhandle (EventDispatchers|CreateEvent :!event BP_HeistMusicManager_AutoGenFunc)))

(fn EndLockdownMusic
  (bind _musicfadetime (Variables|Default|GetMusicFadeTime))
  (bind _lockdowntrack (Variables|Default|GetLockdownTrack)
    (:then
      (if (Audio|Components|Audio|IsPlaying _lockdowntrack)
        (Audio|Components|Audio|FadeOut _lockdowntrack _musicfadetime 0.0 "Linear" :!class /Script/Engine.AudioComponent)))
    (:else
      (bind _lockdownoutro (Variables|Default|GetLockdownOutro)
        (:then
          (if (Audio|Components|Audio|IsPlaying _lockdownoutro)
            (Audio|Components|Audio|FadeOut _lockdownoutro _musicfadetime 0.0 "Linear" :!class /Script/Engine.AudioComponent)))
        (:else)))))

(fn FadeInHeistingMusic
  (bind _audio_heisting (Variables|Default|GetAudio_Heisting)
    (:then
      (if (not (Audio|Components|Audio|IsPlaying _audio_heisting))
        (Audio|Components|Audio|FadeIn _audio_heisting (Variables|Default|GetMusicFadeTime) 1.0 0.0 "Linear" :!class /Script/Engine.AudioComponent)
        (Audio|AddEventbyComponent (GameInstanceSubsystems|GetMusicEventSubsystem) _audio_heisting "(TagName=\"MusicEventSubsystem.Event.POI\")" "(GameplayTags=)" 0 "")))
    (:else)))

(fn FadeInLockdownMusic
  (bind _clydelockdowntrack (Variables|Default|GetClydeLockdownTrack)
    (:then
      (if (not (Audio|Components|Audio|IsPlaying _clydelockdowntrack))
        (Audio|Components|Audio|FadeIn _clydelockdowntrack (Variables|Default|GetMusicFadeTime) 1.0 0.0 "Linear" :!class /Script/Engine.AudioComponent)
        (Audio|AddEventbyComponent (GameInstanceSubsystems|GetMusicEventSubsystem) _clydelockdowntrack "(TagName=\"MusicEventSubsystem.Event.POI.Special\")" "(GameplayTags=)" 0 "")))
    (:else)))

(fn FadeOutHeistingMusic
  (Utilities|FlowControl|Sequence
    (:then_0
      (bind _audio_heisting (Variables|Default|GetAudio_Heisting)
        (:then
          (if (Audio|Components|Audio|IsPlaying _audio_heisting)
            (Audio|Components|Audio|FadeOut _audio_heisting (Variables|Default|GetMusicFadeTime) 0.0 "Linear" :!class /Script/Engine.AudioComponent)))
        (:else)))
    (:then_1
      (Audio|RemoveEvent (GameInstanceSubsystems|GetMusicEventSubsystem) (Variables|Default|GetAudio_Heisting)))))

(fn FadeOutIfPlaying
  (params (Component :type Audio%20Component%20Object%20Reference))
  (if (Audio|Components|Audio|IsPlaying Component)
    (Audio|Components|Audio|FadeOut Component (Variables|Default|GetMusicFadeTime) 0.0 "Linear" :!class /Script/Engine.AudioComponent)
    (Audio|RemoveEvent (GameInstanceSubsystems|GetMusicEventSubsystem) Component)))

(fn FadeOutLockdownMusic
  (Utilities|FlowControl|Sequence
    (:then_0
      (bind _clydelockdowntrack (Variables|Default|GetClydeLockdownTrack)
        (:then
          (if (Audio|Components|Audio|IsPlaying _clydelockdowntrack)
            (Audio|Components|Audio|FadeOut _clydelockdowntrack (Variables|Default|GetMusicFadeTime) 0.0 "Linear" :!class /Script/Engine.AudioComponent)))
        (:else)))
    (:then_1
      (Audio|RemoveEvent (GameInstanceSubsystems|GetMusicEventSubsystem) (Variables|Default|GetClydeLockdownTrack)))))

(fn GetAndRegisterBaseManagers
  (bind _returnvalue (WorldSubsystems|GetFortGlobalActorRegistrySubsystem))
  (bind _basemanagerregistrationtag (Variables|BaseRegistration|GetBaseManagerRegistrationTag))
  ;[0.4973 0.715 1] GetAndRegisterBaseManagers
  ; Get and cache all the base managers so they can be used to determine the current base
  (Networking|SwitchHasAuthority
    (:Authority
      (bind _outactors (Class|FortGlobalActorRegistrySubsystem|GetRegisteredActors _returnvalue _basemanagerregistrationtag)) ; get all the current registered actors
      (for _array_element _outactors
        ;;[0.4973 0.715 1] Add all found actors - run additional validation if necessary
        (BaseRegistration|CacheRegisteredManager :RegisteredActor _array_element))
      ;;[0.4973 0.715 1] Bind to Data Registry object - Will register any base managers that are added to the registry object after the initial run - this is likely because they are spacially loaded
      (bind _returnvalue_2 (Class|FortGlobalActorRegistrySubsystem|GetActorRegistryDataforGameplayTag _returnvalue _basemanagerregistrationtag))
      (Utilities|IsValid _returnvalue_2
        (:Is%20Valid
          (bind _output_get (Variables|BaseRegistration|SetCachedDataRegistry _returnvalue_2))
          (EventDispatchers|BindEventtoOnActorRegistered _output_get (EventDispatchers|CreateEvent :!event CacheRegisteredManager)))))))

(fn HandleNewMusicEvent
  (params (InstigatorPawn :type Fort%20Player%20Pawn%20Object%20Reference) (NewMusicEvent :type Enum_HeistAudioStatus%20Enum :default "NewEnumerator3") (HeistBaseIdentifierTag :type Gameplay%20Tag%20Structure) (HeistBaseManagerRef :type BGA%20Clyde%20Base%20Manager%20Object%20Reference))
  (bind _output_get (Variables|Conditions|SetCurrentHeistState NewMusicEvent))
  (switch Utilities|FlowControl|Switch|SwitchonEnum_HeistAudioStatus _output_get
    (:NewEnumerator3
      (CallFunction|StopLockdownMusic))
    (:NewEnumerator0
      _)
    (:NewEnumerator1
      _))
  (CallFunction|TryPlayStinger) ; Try to play a simple stinger sound
)

(fn InitOnCurrentBaseLockdown
  ;[0.2924 0.5475 0.7292] Init On Current Base Lockdown
  ; When a new base is registered and entered, check lockdown. If not already in lockdown, bind to the Lockdown delegate [Client] to get the trigger
  (Utilities|FlowControl|Sequence
    (:then_0
      (bind _currentbasemanager (Variables|BaseRegistration|GetCurrentBaseManager)
        (:then
          (if (Variables|/Clyde/Audio/Items/BGA_ClydeBaseManager.BGA_ClydeBaseManager_C|GetbInLockdown _currentbasemanager)
            (return (Variables|/Clyde/Audio/Items/BGA_ClydeBaseManager.BGA_ClydeBaseManager_C|GetbInLockdown _currentbasemanager))))
        (:else)))
    (:then_1
      (return false))))

(fn LockdownQuantizedEvents
  (Variables|Conditions|SetStingerRetriggerDelay (Variables|Conditions|GetStingerRetriggerDelay)))

(fn MapAndRequestHeistStatusChange
  (params (AggroBotsMap :type Map%20of%20Actor%20Object%20References%20to%20Enum_HeistAudioStatus%20Enums) (AlertedActor :type Actor%20Object%20Reference) (Heist%20Status :type Enum_HeistAudioStatus%20Enum :default "NewEnumerator3"))
  ;[1 1 1 0.5] 1) log of what bots desire which heist status; 2) updates the heist state if requirements are met; 3) returns true if the heist state was changed.
  (Utilities|FlowControl|Sequence
    (:then_0
      (if (Utilities|Enum|Equal(Enum) Heist%20Status "NewEnumerator2")
        (Utilities|Map|Add AggroBotsMap AlertedActor Heist%20Status)
        (else
          (Utilities|Map|Remove AggroBotsMap AlertedActor)))
      (if (not (Utilities|Enum|Equal(Enum) (Variables|Conditions|GetCurrentHeistState) Heist%20Status))
        (switch Utilities|FlowControl|Switch|SwitchonEnum_HeistAudioStatus Heist%20Status
          (:NewEnumerator3
            (if (Utilities|Map|IsEmpty AggroBotsMap)
              _))
          (:NewEnumerator2
            (if (== (Utilities|Map|Length AggroBotsMap) 1)
              _)))
        (Variables|Conditions|SetCurrentHeistState Heist%20Status)
        (return true)))
    (:then_1
      (return false))))

(fn PlayLockdownMusic
  (params (Tag :type Gameplay%20Tag%20Structure))
  (CallFunction|AudioDebug "Play Lockdown Music")
  (if (GameplayTags|MatchesTag Tag (Variables|BaseRegistration|GetCurrentHeistBaseTag) true)
    (Quartz|CreateLockdownClock)
    (bind _returnvalue (Audio|CreateSound2D (Variables|Music|GetMxLockdownIntro)))
    (bind _output_get (Variables|Default|SetLockdownIntro _returnvalue))
    (Audio|Components|Audio|PlayQuantized _output_get (Variables|Quartz|GetLockdownClock) (Variables|Quartz|GetIntroMusicBoundary))
    (Audio|AddEventbyComponent (GameInstanceSubsystems|GetMusicEventSubsystem) (Variables|Default|GetLockdownIntro) "(TagName=\"MusicEventSubsystem.Event.Emote.Gameplay\")" "(GameplayTags=)" 0 "")
    (bind _returnvalue_2 (Audio|CreateSound2D (Variables|Music|GetMxLockdown)))
    (bind _output_get_2 (Variables|Default|SetLockdownTrack _returnvalue_2))
    (Audio|Components|Audio|PlayQuantized _output_get_2 (Variables|Quartz|GetLockdownClock) (Variables|Quartz|GetMainTrackMusicBoundary))
    (Audio|AddEventbyComponent (GameInstanceSubsystems|GetMusicEventSubsystem) (Variables|Default|GetLockdownTrack) "(TagName=\"MusicEventSubsystem.Event.Emote.Gameplay\")" "(GameplayTags=)" 0 "")
    (return)))

(fn PlayLockdownOutro
  (bind _returnvalue (Audio|CreateSound2D (Variables|Music|GetMxLockdownOutro)))
  (bind _output_get (Variables|Default|SetLockdownOutro _returnvalue))
  (Audio|AddEventbyComponent (GameInstanceSubsystems|GetMusicEventSubsystem) _output_get "(TagName=\"MusicEventSubsystem.Event.Emote.Gameplay\")" "(GameplayTags=)" 0 "")
  (Audio|Components|Audio|PlayQuantized _output_get (Variables|Quartz|GetLockdownClock) (Variables|Quartz|GetOutroMusicBoundary)))

(fn SetCurrentBaseManager
  (params (HeistBaseTagID :type Gameplay%20Tag%20Structure))
  (Utilities|FlowControl|Sequence
    (:then_0
      (for _array_element (Variables|BaseRegistration|GetBaseManagers)
        (Utilities|IsValid _array_element
          (:Is%20Valid
            (if (GameplayTags|MatchesTag (Variables|/Clyde/Audio/Items/BGA_ClydeBaseManager.BGA_ClydeBaseManager_C|GetHeistBaseId%20Tag%20Const _array_element) HeistBaseTagID true)
              (Variables|BaseRegistration|SetCurrentBaseManager _array_element)
              ;[1 0.5249 0.9376] Return true - a valid base was found and cached
              (Variables|BaseRegistration|SetbShouldFindBaseOnRegistration false) ; when new bases are registered, \nshould we try to find a new base?
              (return true))))))
    (:then_1
      ;[1 0.5249 0.9376] Valid base manager not found
      ; Using a sequence ensures that the return node is called on all invalid execution paths
      (Variables|BaseRegistration|SetbShouldFindBaseOnRegistration true) ; when new bases are registered, \nshould we try to find a new base?
      (return false))))

(fn SetDataForCurrentPOI
  (params (Actor :type Actor%20Object%20Reference) (Current%20Base%20Tag :type Gameplay%20Tag%20Structure))
  (Variables|BaseRegistration|SetCurrentHeistBaseTag Current%20Base%20Tag :!pinned true) ; cache the current base tag so that it \ncan be checked in case of a late \nBaseManager registration
  (Variables|Conditions|SetbInsideCurrentPOI true))

(fn SetLocalPawnRef
  (bind _asfort_player_pawn_athena (Utilities|Casting|CastToFortPlayerPawnAthena (Game|Player|GetPlayerPawn (Fortnite|GetFirstLocalFortPlayerController true true))))
  (if (Pawn|IsLocallyViewed _asfort_player_pawn_athena)
    (bind _output_get (Variables|Default|SetPawnRef _asfort_player_pawn_athena))
    (Game|BindEventtoOnDestroyed _output_get (EventDispatchers|CreateEvent :!event FadeMusicOutWhenPlayerDies))))

(fn StopLockdownMusic
  (CallFunction|AudioDebug "Stop Lockdown Music")
  (bind _currentheiststinger (Variables|Default|GetCurrentHeistStinger)
    (:then
      (if (Audio|Components|Audio|IsPlaying _currentheiststinger)
        (Audio|Components|Audio|FadeOut _currentheiststinger 0.1 0.0 "Linear" :!class /Script/Engine.AudioComponent)))
    (:else)))

(fn ToggleMusicGate
  (Variables|Conditions|SetbMusicGateIsOpen (not (Variables|Conditions|GetbMusicGateIsOpen))))

(fn TrackPOILockdowns
  (params (HeistBaseID :type Gameplay%20Tag%20Structure))
  (switch int (Utilities|Array|FindItem (Variables|Tags|GetPOITagArray) HeistBaseID)
    (:0
      (Variables|Conditions|SetbEstateHasBeenLockedDown true))
    (:1
      (Variables|Conditions|SetbHotelHasBeenLockedDown true))
    (:2
      (Variables|Conditions|SetbResortHasBeenLockedDown true))))

(fn TryPlayStinger
  (bind _bmusicgateisopen (Variables|Conditions|GetbMusicGateIsOpen))
  (bind _nextstingertoplay (Variables|Music|GetNextStingerToPlay))
  (if _bmusicgateisopen
    (Variables|Conditions|SetbMusicGateIsOpen (not _bmusicgateisopen))
    (if (Utilities|Enum|NotEqual(Enum) (Variables|Conditions|GetCurrentHeistState) "NewEnumerator2")
      (bind _currentheiststinger (Variables|Default|GetCurrentHeistStinger)
        (:then
          (Audio|Components|Audio|SetSound _currentheiststinger _nextstingertoplay)
          (Audio|Components|Audio|Play (Variables|Default|GetCurrentHeistStinger) 0.0)
          (bind _returnvalue (Utilities|Time|SetTimerbyEvent (EventDispatchers|CreateEvent :!event ToggleMusicGate) (Variables|Conditions|GetStingerRetriggerDelay)))
          (Variables|Default|SetHeistMusicHandle _returnvalue))
        (:else
          (bind _returnvalue_2 (Audio|CreateSound2D _nextstingertoplay))
          (bind _output_get (Variables|Default|SetCurrentHeistStinger _returnvalue_2)))))))

(fn UpdateHeistMusicAndTagToCurrent
  (bind _currentheiststate (Variables|Conditions|GetCurrentHeistState))
  (bind _musictagmap (Variables|Maps|GetMusicTagMap))
  (bind (_value _returnvalue) (Utilities|Map|Find _musictagmap _currentheiststate))
  (Utilities|FlowControl|Sequence
    (:then_0
      ;[0.1407 0.1494 1] Set Music Tag
      (if _returnvalue
        (Variables|Default|SetCurrentMusicEventTag _value)))
    (:then_1
      ;[0.1407 0.1494 1] Set Next Stinger To Play
      (Variables|Music|SetNextStingerToPlay (Utilities|Select :Index (Variables|Conditions|GetCurrentHeistState) :NewEnumerator3 (Variables|Music|GetSting_Pass) :NewEnumerator0 (Variables|Music|GetSting_Pass) :NewEnumerator1 (Variables|Music|GetSting_Fail))))))

(fn ConstructionScript)
