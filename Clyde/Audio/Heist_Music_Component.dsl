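; Heist music component for Clyde that relays heist audio status and audio volume changes to BP_ClydeMusicManager.
; Once playlist data is ready on cosmetic-capable instances, it binds AudioGameplayVolume listener enter/exit events and finds the manager actor.
; Music events reach the manager only for valid, locally controlled pawns.
; Volume toggles send the stored heist status without a pawn, then request a music update and stinger directly.
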(event EventBeginPlay
  ;[0 1 0] Begin Play - Bind to AudioGameplayVolume events if valid. Set the reference to the heist music manager
  (Networking|CanExecuteCosmeticEvents
    (:True
      (Class|FortAsyncActionCurrentPlaylistReady|PlaylistDataReadyAsync
        (:OnReady
          (Utilities|FlowControl|Sequence
            (:then_0
              (bind _agv (Variables|Default|GetAGV)
                (:then
                  (Events|BindEventtoOnListenerEnterEvent _agv (EventDispatchers|CreateEvent :!event OnListenerEnter))
                  (Events|BindEventtoOnListenerExitEvent _agv (EventDispatchers|CreateEvent :!event OnListenerExit)))
                (:else)))
            (:then_1
              (CallFunction|SetHeistMusicManagerReference))))))))

(event Custom|SendMusicEventToManager
  (params (InstigatorPawn :type Fort%20Player%20Pawn%20Object%20Reference) (HeistMusicEvent :type Enum_HeistAudioStatus%20Enum :default "NewEnumerator3") (HeistBaseIdentiferTag :type Gameplay%20Tag%20Structure) (HeistBaseID :type BGA%20Clyde%20Base%20Manager%20Object%20Reference))
  (bind _heistmusicmanager (Variables|Default|GetHeistMusicManager))
  ;[1 0 1] Send information to the manager which will determine if we need to play/stop/update music
  (if (and (Utilities|IsValid _heistmusicmanager :!class /Script/Engine.KismetSystemLibrary) (Utilities|IsValid InstigatorPawn :!class /Script/Engine.KismetSystemLibrary))
    (if (Pawn|IsLocallyControlled InstigatorPawn)
      (Class|BPClydeMusicManager|HandleNewMusicEvent _heistmusicmanager InstigatorPawn HeistMusicEvent HeistBaseIdentiferTag 0))))

(event AudioGameplayVolume|EventOnToggledOn
  ;[0.2344 0.3878 1] Audio Gameplay Volume
  (CallFunction|SendMusicEventtoManager 0 (Variables|Default|GetCurrentHeistStatus))
  (CallFunction|RequestMusicPlayback "NewEnumerator0")
  (Development|PrintString "Heist Audio: Player entered audio gameplay volume"))

(event AudioGameplayVolume|EventOnToggledOff
  ;[0.2344 0.3878 1] Audio Gameplay Volume
  (CallFunction|SendMusicEventtoManager 0 (Variables|Default|GetCurrentHeistStatus))
  (CallFunction|RequestMusicPlayback "NewEnumerator3")
  (Development|PrintString "Heist Audio: Player exited audio gameplay volume"))

(fn HandleHeistMusicEvent
  (params (FortPlayerPawn :type Fort%20Player%20Pawn%20Athena%20Object%20Reference) (HeistEvent :type Enum_HeistAudioStatus%20Enum :default "NewEnumerator3") (HeistBaseTag :type Gameplay%20Tag%20Structure))
  (Networking|CanExecuteCosmeticEvents
    (:True
      (Utilities|IsValid FortPlayerPawn
        (:Is%20Valid
          (if (Pawn|IsLocallyControlled FortPlayerPawn)
            (bind _heistmusicmanager (Variables|Default|GetHeistMusicManager)
              (:then
                (Class|BPClydeMusicManager|UpdateHeistMusicandTagtoCurrent _heistmusicmanager))
              (:else))))))))

(fn RequestMusicPlayback
  (params (HeistAudioState :type Enum_HeistAudioStatus%20Enum :default "NewEnumerator3"))
  (bind _heistmusicmanager (Variables|Default|GetHeistMusicManager))
  (Networking|CanExecuteCosmeticEvents
    (:True
      (Class|BPClydeMusicManager|UpdateHeistMusicandTagtoCurrent _heistmusicmanager)
      (Class|BPClydeMusicManager|TryPlayStinger _heistmusicmanager))))

(fn SetHeistMusicManagerReference
  (Networking|CanExecuteCosmeticEvents
    (:True
      (bind _returnvalue (Actor|GetActorOfClass "/Clyde/Audio/BP_ClydeMusicManager.BP_ClydeMusicManager_C"))
      (Utilities|IsValid _returnvalue
        (:Is%20Valid
          (Variables|Default|SetHeistMusicManager _returnvalue))
        (:Is%20Not%20Valid
          _)))
    (:False
      _))
  ;[1 0 0] No manager error
  (Fortnite|FortShippingLog self "Failed to find heist music manager" true))

(fn SetHeistStatus
  (params (CurrentHeistStatus :type Enum_HeistAudioStatus%20Enum :default "NewEnumerator3") (bIsLockdown :type Boolean))
  (Variables|Default|SetbIsLockdown bIsLockdown)
  (bind _output_get (Variables|Default|SetCurrentHeistStatus CurrentHeistStatus))
  (return _output_get))
