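; Clyde Blueprint that manages World Partition data layers for seasonal map updates and the Kelp Linen day configurations.
; SetSeasonDataLayers maps an index to a season and activates or unloads its listed layers.
; LoadKelpLinenConfiguration picks one of three day layer sets from a curve table value, while KelpLinenTest takes the index directly.
; OnRep_KelpLoaded invalidates all runtime virtual texture volumes on cosmetic-capable machines; RefreshVT refreshes a tagged one using console commands.
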(fn FlushRVT
  (bind _outactors (Actor|GetAllActorsOfClass "/Script/Engine.RuntimeVirtualTextureVolume"))
  (for _array_element _outactors
    (VirtualTexture|Invalidate (Variables|/Script/Engine.RuntimeVirtualTextureVolume|GetVirtualTextureComponent _array_element) (Math|BoxSphereBounds|MakeBoxSphereBounds (Collision|GetComponentBounds (Variables|/Script/Engine.RuntimeVirtualTextureVolume|GetVirtualTextureComponent _array_element)) (.BoxExtent (Collision|GetComponentBounds (Variables|/Script/Engine.RuntimeVirtualTextureVolume|GetVirtualTextureComponent _array_element))) (.SphereRadius (Collision|GetComponentBounds (Variables|/Script/Engine.RuntimeVirtualTextureVolume|GetVirtualTextureComponent _array_element)))) "High")
    (Development|PrintString "RVT Flush!" true true "(R=0.223878,G=1.000000,B=0.000000,A=1.000000)" 1e+01)))

(fn KelpLinenTest
  (params (Selection :type Integer))
  (bind _returnvalue (WorldPartition|GetDataLayerManager))
  (switch int Selection
    (:0
      ; Week 4 - Day 1
      (for _array_element (Variables|Default|GetKelpLinen_Day1_ToLoad)
        (DataLayers|SetDataLayerInstanceRuntimeState _returnvalue (DataLayers|GetDataLayerInstancefromAsset _returnvalue _array_element) "Activated" false))
      (for _array_element_2 (Variables|Default|GetKelpLinen_Day1_ToUnload)
        (DataLayers|SetDataLayerInstanceRuntimeState _returnvalue (DataLayers|GetDataLayerInstancefromAsset _returnvalue _array_element_2) "Unloaded" false)))
    (:1
      ; Week 4 - Day 3
      (for _array_element_3 (Variables|Default|GetKelpLinen_Day3_ToLoad)
        (DataLayers|SetDataLayerInstanceRuntimeState _returnvalue (DataLayers|GetDataLayerInstancefromAsset _returnvalue _array_element_3) "Activated" false))
      (for _array_element_4 (Variables|Default|GetKelpLinen_Day3_ToUnload)
        (DataLayers|SetDataLayerInstanceRuntimeState _returnvalue (DataLayers|GetDataLayerInstancefromAsset _returnvalue _array_element_4) "Unloaded" false)))
    (:2
      ; Week 4 - Day 5
      (for _array_element_5 (Variables|Default|GetKelpLinen_Day5_ToLoad)
        (DataLayers|SetDataLayerInstanceRuntimeState _returnvalue (DataLayers|GetDataLayerInstancefromAsset _returnvalue _array_element_5) "Activated" false))
      (for _array_element_6 (Variables|Default|GetKelpLinen_Day5_ToUnload)
        (DataLayers|SetDataLayerInstanceRuntimeState _returnvalue (DataLayers|GetDataLayerInstancefromAsset _returnvalue _array_element_6) "Unloaded" false))))
  (CallFunction|KelpLinenLoadedonServer))

(fn LoadKelpLinenConfiguration
  (bind _outxy (DataTable|EvaluateCurveTableRow "/Clyde/Datatables/ClydeCompositeGD_S3.ClydeCompositeGD_S3" "Default.KelpLinen.Configuration")
    (:RowFound
      (bind _returnvalue (WorldPartition|GetDataLayerManager))
      (Development|PrintString (Utilities|String|Append "Kelp Linen Index " (Utilities|String|ToString(Integer) (Math|Float|Truncate _outxy))) true true "(R=1.000000,G=0.081353,B=0.000000,A=1.000000)" 6e+01)
      (Fortnite|FortShippingLog self (Utilities|String|Append "Kelp Linen Index " (Utilities|String|ToString(Integer) (Math|Float|Truncate _outxy))))
      (switch int (Math|Float|Truncate _outxy)
        (:0
          ; Week 4 - Day 1
          (for _array_element (Variables|Default|GetKelpLinen_Day1_ToLoad)
            (DataLayers|SetDataLayerInstanceRuntimeState _returnvalue (DataLayers|GetDataLayerInstancefromAsset _returnvalue _array_element) "Activated" false))
          (for _array_element_2 (Variables|Default|GetKelpLinen_Day1_ToUnload)
            (DataLayers|SetDataLayerInstanceRuntimeState _returnvalue (DataLayers|GetDataLayerInstancefromAsset _returnvalue _array_element_2) "Unloaded" false)))
        (:1
          ; Week 4 - Day 3
          (for _array_element_3 (Variables|Default|GetKelpLinen_Day3_ToLoad)
            (DataLayers|SetDataLayerInstanceRuntimeState _returnvalue (DataLayers|GetDataLayerInstancefromAsset _returnvalue _array_element_3) "Activated" false))
          (for _array_element_4 (Variables|Default|GetKelpLinen_Day3_ToUnload)
            (DataLayers|SetDataLayerInstanceRuntimeState _returnvalue (DataLayers|GetDataLayerInstancefromAsset _returnvalue _array_element_4) "Unloaded" false)))
        (:2
          ; Week 4 - Day 5
          (for _array_element_5 (Variables|Default|GetKelpLinen_Day5_ToLoad)
            (DataLayers|SetDataLayerInstanceRuntimeState _returnvalue (DataLayers|GetDataLayerInstancefromAsset _returnvalue _array_element_5) "Activated" false))
          (for _array_element_6 (Variables|Default|GetKelpLinen_Day5_ToUnload)
            (DataLayers|SetDataLayerInstanceRuntimeState _returnvalue (DataLayers|GetDataLayerInstancefromAsset _returnvalue _array_element_6) "Unloaded" false)))))
    (:RowNotFound
      (Fortnite|FortShippingLog self "Couldn't find row for kelp linen config"))))

(fn OnRep_KelpLoaded
  (Networking|CanExecuteCosmeticEvents
    (:True
      (CallFunction|FlushRVT))))

(fn RefreshVT
  (bind _outactors (Actor|GetAllActorsOfClasswithTag "/Script/Engine.RuntimeVirtualTextureVolume" "RVT"))
  (VirtualTexture|Invalidate (Variables|/Script/Engine.RuntimeVirtualTextureVolume|GetVirtualTextureComponent (Utilities|Array|Get(acopy) _outactors)) (Math|BoxSphereBounds|MakeBoxSphereBounds (Collision|GetComponentBounds (Variables|/Script/Engine.RuntimeVirtualTextureVolume|GetVirtualTextureComponent (Utilities|Array|Get(acopy) _outactors))) (.BoxExtent (Collision|GetComponentBounds (Variables|/Script/Engine.RuntimeVirtualTextureVolume|GetVirtualTextureComponent (Utilities|Array|Get(acopy) _outactors)))) (.SphereRadius (Collision|GetComponentBounds (Variables|/Script/Engine.RuntimeVirtualTextureVolume|GetVirtualTextureComponent (Utilities|Array|Get(acopy) _outactors))))) "High")
  (Development|ExecuteConsoleCommand "r.VT.RefreshEntirePageTable 1")
  (Development|ExecuteConsoleCommand "r.VT.ForceContinuousUpdate 1")
  (Development|ExecuteConsoleCommand "streamout 20000"))

(fn SetSeasonDataLayers
  (params (MapIndex :type Integer))
  (Variables|Default|SetCurrentSeason (Utilities|Array|Get(acopy) (Variables|Default|GetMapUpdateSeasonArray) MapIndex))
  (for _array_element (Variables|Default|GetSeasonsDataLayerList)
    (if (Utilities|String|EqualExactly(String) (Variables|Default|GetCurrentSeason) (Utilities|Struct|BreakCalendarDataLayerData _array_element))
      (bind _returnvalue (WorldPartition|GetDataLayerManager))
      (Utilities|FlowControl|Sequence
        (:then_0
          (for _array_element_2 (.EnabledDataLayers_9_41BB9A3E4648805B5AF70AA728288C92 (Utilities|Struct|BreakCalendarDataLayerData _array_element))
            (DataLayers|SetDataLayerInstanceRuntimeState _returnvalue (DataLayers|GetDataLayerInstancefromAsset _returnvalue _array_element_2) "Activated" false)))
        (:then_1
          (for _array_element_3 (.DisabledDataLayers_10_8DE59E3B474C88ED2E7988B185572444 (Utilities|Struct|BreakCalendarDataLayerData _array_element))
            (DataLayers|SetDataLayerInstanceRuntimeState _returnvalue (DataLayers|GetDataLayerInstancefromAsset _returnvalue _array_element_3) "Unloaded" false)))))))

(fn ConstructionScript
  (|CallParentFunction))
