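; Athena supply drop variant (Clyde) that checks on BeginPlay whether it spawned above water and relocates itself.
; It line-traces straight down from the crate mesh and casts the hit actor to FortWaterBodyBP.
; Following that cast, it moves the actor to the current safe zone phase's location plus a random horizontal offset.
; Overlap, tick and construction script add no logic beyond the construction parent call.
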(event EventBeginPlay
  (|CallParentFunction)
  (bind _outhit (Collision|LineTraceByChannel (Transformation|GetWorldLocation (Variables|Default|GetCrateStaticMesh)) (Math|Vector|MakeVector (.x (Transformation|GetWorldLocation (Variables|Default|GetCrateStaticMesh))) (.y (Transformation|GetWorldLocation (Variables|Default|GetCrateStaticMesh))) (- (.z (Transformation|GetWorldLocation (Variables|Default|GetCrateStaticMesh))) 1e+04))))
  (bind (_outhit_bblockinghit _outhit_binitialoverlap _outhit_time _outhit_distance _outhit_location _outhit_impactpoint _outhit_normal _outhit_impactnormal _outhit_physmat _outhit_hitactor) (Collision|BreakHitResult _outhit))
  (Utilities|Casting|CastToFortWaterBodyBP _outhit_hitactor)
  (bind _returnvalue (SafeZone|GetCurrentSafeZonePhase))
  (bind _outlocation (SafeZone|GetSafeZoneLocation _returnvalue))
  (Transformation|SetActorLocation (+ _outlocation (Utilities|RandomVectorinRange "-5000.000000,-5000.000000,0.000000" "5000.000000,5000.000000,0.000000"))))

(event Collision|EventActorBeginOverlap
  (params (OtherActor :type Actor%20Object%20Reference)))

(event EventTick
  (params (DeltaSeconds :type Float%20%28single-precision%29)))

(fn ConstructionScript
  (|CallParentFunction))
