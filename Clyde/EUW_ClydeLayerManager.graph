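# Editor utility widget for Clyde that switches the level's calendar data layers between seasonal weeks and KelpLinen day variants.
# It finds the single BP_Clyde_DataLayerManager in the level, and week buttons match a season name against its list to load and show or unload and hide layers.
# An optional checkbox unloads other seasons' layers, and swaps refresh virtual texture pages via console commands.
# KelpLinen buttons apply the week 3Part1 season, then day-specific load and unload lists.
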EditorUtilityWidgetBlueprint /Clyde/EUW_ClydeLayerManager
graphs [
 Call Season Swap {nodes [
   n1 = Call Season Swap!() -> (then, Week:String) {is=title K2Node_FunctionEntry}
   > Set Season(Season:String=@n1) -> (Output_Get:String)
   n2 = Get LayerManager() -> (LayerManager:BP Clyde Data Layer Manager Object Reference)
   > n3 = Is Valid(InputObject=@n2)
   GetLayerManager(execute=@n3.Is Not Valid)
   n4 = For Each Loop(
    Exec=@n3.Is Valid
    Array:Array of Calendar Data Layer Data Structures=Get SeasonsDataLayerList!(self:BP Clyde Data Layer Manager Object Reference=@n2) -> (SeasonsDataLayerList:Array of Calendar Data Layer Data Structures)
   ) -> (Array Element:Calendar Data Layer Data Structure)
   n5 = Branch(execute=@n4.Completed, Condition=IsChecked(self=Get UnloadCheckBox() -> (UnloadCheckBox:Editor Utility Check Box Object Reference)))
   > n6 = For Each Loop(Array=Get SeasonsDataLayerList(self=@n2))
   ExecuteConsoleCommand(execute=[@n5.else, @n6.Completed], Command=r.VT.RefreshEntirePageTable 0)
   n7 = Get DataLayerEditorSubsystem() -> (ReturnValue:Data Layer Editor Subsystem Object Reference) {is=title K2Node_GetEditorSubsystem}
   n8 = For Each Loop(
    Exec=@n6.LoopBody
    Array:Array of Data Layer Instance Object References=GetDataLayerInstances!(
     self:Data Layer Editor Subsystem Object Reference=@n7
     DataLayerAssets:Array of Data Layer Asset Object References=Break Calendar Data Layer Data!(CalendarDataLayerData:"Calendar Data Layer Data Structure (by ref)"=@n6) -> (
      EventName_2_#1:String
      EnabledDataLayers_9_#2:Array of Data Layer Asset Object References
      DisabledDataLayers_10_#3:Array of Data Layer Asset Object References
     ) {is=title K2Node_BreakStruct}.EnabledDataLayers_9_#2
    ) -> (ReturnValue:Array of Data Layer Instance Object References)
   ) -> (Array Element:Data Layer Instance Object Reference)
   n9 = Contains Item(
    TargetArray:Array of Data Layer Instance Object References=Get EnabledLayers() -> (EnabledLayers:Array of Data Layer Instance Object References)
    ItemToFind:"Data Layer Instance Object Reference (by ref)"=@n8
   )
   > Branch(Condition=AND Boolean(A=NOT Boolean(A=@n9), B=IsVisible(self:Data Layer Instance Object Reference=@n8)))
   > SetDataLayerIsLoadedInEditor!(
    execute
    self:Data Layer Editor Subsystem Object Reference=@n7
    DataLayer:Data Layer Instance Object Reference=@n8
    bIsLoadedInEditor:bool=false
    bIsFromUserChange:bool=false
   ) -> (then, ReturnValue:bool)
   > SetDataLayerVisibility!(execute, self:Data Layer Editor Subsystem Object Reference=@n7, DataLayer:Data Layer Instance Object Reference=@n8, bIsVisible:bool=false) -> (then)
   n10 = Break Calendar Data Layer Data(CalendarDataLayerData=@n4)
   Branch(execute=@n4.LoopBody, Condition="Equal Exactly (String)"(A=Get Season() -> (Season:String), B=@n10))
   > n11 = Sequence!(execute) -> (then_0, then_1, then_2:Exec, then_3:Exec)
   n12 = Get DataLayerEditorSubsystem()
   > n13 = Set EnabledLayers(EnabledLayers:Array of Data Layer Instance Object References=GetDataLayerInstances(self=@n12, DataLayerAssets=@n10.EnabledDataLayers_9_#2)) -> (Output_Get:Array of Data Layer Instance Object References)
   > SetDataLayersIsLoadedInEditor!(
    execute
    self:Data Layer Editor Subsystem Object Reference=@n12
    DataLayers:Array of Data Layer Instance Object References=@n13
    bIsLoadedInEditor:bool=true
    bIsFromUserChange:bool=false
   ) -> (then, ReturnValue:bool)
   > n14 = SetDataLayersVisibility!(
    execute
    self:Data Layer Editor Subsystem Object Reference=@n12
    DataLayers:Array of Data Layer Instance Object References=@n13
    bIsVisible:bool=true
   ) -> (then)
   ExecuteConsoleCommand(execute=[@n16.then, @n14.then], Command=r.VT.RefreshEntirePageTable 1)
   > ExecuteConsoleCommand(Command=r.VT.ForceContinuousUpdate 1)
   > ExecuteConsoleCommand(Command=streamout 2000)
   n15 = GetDataLayerInstances(self=@n12, DataLayerAssets=@n10.DisabledDataLayers_10_#3)
   SetDataLayersIsLoadedInEditor(execute=@n11.then_1, self=@n12, DataLayers=@n15, bIsLoadedInEditor=false, bIsFromUserChange=false)
   > n16 = SetDataLayersVisibility(self=@n12, DataLayers=@n15, bIsVisible=false)
  ]}
 EventGraph {nodes [
   "On Clicked (WeekButton_1)"() {is=title K2Node_ComponentBoundEvent}
   n1 = Get WeekName() -> (WeekName:String)
   > Call Season Swap!(execute, self, Week:String=Append(A=@n1, B=1)) -> (then) {is=title K2Node_CallFunction}
   > n2 = Set SeasonText(SeasonText:double=1) -> (Output_Get:double)
   > Set Text(
    Text="To Text (Float)"(Value=@n2, MinimumIntegralDigits=1, MaximumIntegralDigits=324, MinimumFractionalDigits=0, MaximumFractionalDigits=3)
    self=Get Week() -> (Week:Text Object Reference)
   )
   > UnloadAllKelpLinen()
   "On Clicked (WeekButton_2)"() {is=title K2Node_ComponentBoundEvent}
   > Call Season Swap(Week=Append(A=@n1, B=2Part1))
   > n3 = Set SeasonText(SeasonText=2)
   > Set Text(
    Text="To Text (Float)"(Value=@n3, MinimumIntegralDigits=1, MaximumIntegralDigits=324, MinimumFractionalDigits=0, MaximumFractionalDigits=3)
    self=Get Week()
   )
   > UnloadAllKelpLinen()
   "On Clicked (KelpButton_1)"() {is=title K2Node_ComponentBoundEvent}
   > LoadKelpLinen!(execute, self, KelpLinenIndex:int=0) -> (then)
   "On Clicked (KelpButton_2)"() {is=title K2Node_ComponentBoundEvent}
   > LoadKelpLinen(KelpLinenIndex=1)
   "On Clicked (KelpButton_3)"() {is=title K2Node_ComponentBoundEvent}
   > LoadKelpLinen(KelpLinenIndex=2)
   "On Clicked (WeekButton_3)"() {is=title K2Node_ComponentBoundEvent}
   > Call Season Swap(Week=Append(A=@n1, B=2Part2))
   > n4 = Set SeasonText(SeasonText=3)
   > Set Text(
    Text="To Text (Float)"(Value=@n4, MinimumIntegralDigits=1, MaximumIntegralDigits=324, MinimumFractionalDigits=0, MaximumFractionalDigits=3)
    self=Get Week()
   )
   > UnloadAllKelpLinen()
   "On Clicked (SelectButton)"() {is=title K2Node_ComponentBoundEvent}
   n5 = Get EditorActorSubsystem()
   > ClearActorSelectionSet(self:Editor Actor Subsystem Object Reference=@n5)
   > SetSelectedLevelActors(self=@n5, ActorsToSelect=Make Array("[0]":Actor Object Reference=Get LayerManager()) -> (Array:Array of Actor Object References))
   "On Clicked (RefreshLayer)"() {is=title K2Node_ComponentBoundEvent}
   > GetLayerManager()
   "On Clicked (RefreshBoW)"() {is=title K2Node_ComponentBoundEvent}
   > ExecuteConsoleCommand(Command=r.Water.WaterInfo.ForceUpdateWaterInfoNextFrames 1)
   "On Clicked (WeekButton_4)"() {is=title K2Node_ComponentBoundEvent}
   > Call Season Swap(Week=Append(A=@n1, B=3Part1))
   > Set SeasonText(SeasonText=4)
   > LoadKelpLinen(KelpLinenIndex=0)
   "On Clicked (UnloadButton)"() {is=title K2Node_ComponentBoundEvent}
   > n6 = For Each Loop(Array:Array of Calendar Data Layer Data Structures=Get SeasonsDataLayerList(self=Get LayerManager())) -> (Array Element:Calendar Data Layer Data Structure)
   > UnloadSeasons!(execute, self, DataLayerData:Calendar Data Layer Data Structure=@n6) -> (then)
   > n7 = Set SeasonText(SeasonText=0)
   > Set Text(
    Text="To Text (Float)"(Value=@n7, MinimumIntegralDigits=1, MaximumIntegralDigits=324, MinimumFractionalDigits=0, MaximumFractionalDigits=3)
    self=Get Week()
   )
   GetLayerManager() {is=title K2Node_CustomEvent}
   > n8 = GetAllLevelActors(self=Get EditorActorSubsystem())
   > n9 = For Each Loop(Array:Array of Actor Object References=@n8) -> (Array Element:Actor Object Reference)
   n10 = Get PossibleLayerManagers() -> (PossibleLayerManagers:Array of BP Clyde Data Layer Manager Object References)
   n11 = Switch on Int!(execute:Exec=@n9.Completed, Selection:int=Length(TargetArray:Array of BP Clyde Data Layer Manager Object References=@n10)) -> (Default:Exec, 0:Exec, 1:Exec)
   Set Text(execute=[@n11.0, @n11.Default], Text="LOCTEXT(\"Error\")", self=Get Week())
   > Set ValidLayerManager(ValidLayerManager:bool=false) -> (Output_Get:bool)
   > n12 = PrintString(InString=There are 0 or more than 1 loaded Data Layer Manager objects in the level., TextColor=(R=1,G=0.9452,B=0.243,A=1), Duration=15)
   Clear(execute=[@n12.then, @n13.then], TargetArray:Array of BP Clyde Data Layer Manager Object References=Get PossibleLayerManagers())
   Set LayerManager(
    execute=@n11.1
    LayerManager:BP Clyde Data Layer Manager Object Reference="Get (a copy)"(Array:Array of BP Clyde Data Layer Manager Object References=@n10) -> (Output:BP Clyde Data Layer Manager Object Reference)
   ) -> (Output_Get:BP Clyde Data Layer Manager Object Reference)
   > Set Text(
    Text="To Text (Float)"(
     Value=Get SeasonText() -> (SeasonText:double)
     MinimumIntegralDigits=1
     MaximumIntegralDigits=324
     MinimumFractionalDigits=0
     MaximumFractionalDigits=3
    )
    self=Get Week()
   )
   > n13 = Set ValidLayerManager(ValidLayerManager=true)
   Is Valid(exec=@n9.LoopBody, InputObject=@n9)
   > Branch(Condition="Equal (Class)"(A=Get Class(Object=@n9), B=&/BP_Clyde_DataLayerManager.BP_Clyde_DataLayerManager_C))
   > Add(
    TargetArray:Array of BP Clyde Data Layer Manager Object References=Get PossibleLayerManagers()
    NewItem:"BP Clyde Data Layer Manager Object Reference (by ref)"=Cast To BP_Clyde_DataLayerManager!(Object=@n9) -> (AsBP Clyde Data Layer Manager:BP Clyde Data Layer Manager Object Reference, bSuccess:bool=~) {is=title K2Node_DynamicCast}
   ) {is=title K2Node_CallArrayFunction}
   Event Construct()
   > Set SeasonText(SeasonText=0)
   > Set ValidLayerManager(ValidLayerManager=false)
   > GetLayerManager() {is=title K2Node_CallFunction}
   > ToggleUIElements()
  ]}
 LoadKelpLinen {nodes [
   n1 = LoadKelpLinen!() -> (then, KelpLinenIndex:int) {is=title K2Node_FunctionEntry}
   > Call Season Swap(Week=Append(A=ClydeSeason, B=3Part1))
   > n2 = Switch on Int!(execute, Selection=@n1) -> (Default, 0, 1, 2:Exec)
   Set Text(execute=@n2.0, Text="To Text (String)"(InString=Append(A="4 Day ", B=1)), self=Get Week())
   n3 = Get DataLayerEditorSubsystem()
   n4 = GetDataLayerInstances(self=@n3, DataLayerAssets=Get KelpLinen_Day1_ToLoad() -> (KelpLinen_Day1_ToLoad:Array of Data Layer Asset Object References))
   > SetDataLayersIsLoadedInEditor(self=@n3, DataLayers=@n4, bIsLoadedInEditor=true, bIsFromUserChange=false)
   > SetDataLayersVisibility(self=@n3, DataLayers=@n4, bIsVisible=true)
   n5 = Get DataLayerEditorSubsystem()
   n6 = GetDataLayerInstances(self=@n5, DataLayerAssets=Get KelpLinen_Day1_ToUnload() -> (KelpLinen_Day1_ToUnload:Array of Data Layer Asset Object References))
   > SetDataLayersIsLoadedInEditor(self=@n5, DataLayers=@n6, bIsLoadedInEditor=false, bIsFromUserChange=false)
   > SetDataLayersVisibility(self=@n5, DataLayers=@n6, bIsVisible=false)
   Set Text(execute=@n2.1, Text="To Text (String)"(InString=Append(A="4 Day ", B=3)), self=Get Week())
   n7 = Get DataLayerEditorSubsystem()
   n8 = GetDataLayerInstances(self=@n7, DataLayerAssets=Get KelpLinen_Day3_ToLoad() -> (KelpLinen_Day3_ToLoad:Array of Data Layer Asset Object References))
   > SetDataLayersIsLoadedInEditor(self=@n7, DataLayers=@n8, bIsLoadedInEditor=true, bIsFromUserChange=false)
   > SetDataLayersVisibility(self=@n7, DataLayers=@n8, bIsVisible=true)
   n9 = Get DataLayerEditorSubsystem()
   n10 = GetDataLayerInstances(self=@n9, DataLayerAssets=Get KelpLinen_Day3_ToUnload() -> (KelpLinen_Day3_ToUnload:Array of Data Layer Asset Object References))
   > SetDataLayersIsLoadedInEditor(self=@n9, DataLayers=@n10, bIsLoadedInEditor=false, bIsFromUserChange=false)
   > SetDataLayersVisibility(self=@n9, DataLayers=@n10, bIsVisible=false)
   Set Text(execute=@n2.2, Text="To Text (String)"(InString=Append(A="4 Day ", B=5)), self=Get Week())
   n11 = Get DataLayerEditorSubsystem()
   n12 = GetDataLayerInstances(self=@n11, DataLayerAssets=Get KelpLinen_Day5_ToLoad() -> (KelpLinen_Day5_ToLoad:Array of Data Layer Asset Object References))
   > SetDataLayersIsLoadedInEditor(self=@n11, DataLayers=@n12, bIsLoadedInEditor=true, bIsFromUserChange=false)
   > SetDataLayersVisibility(self=@n11, DataLayers=@n12, bIsVisible=true)
   n13 = Get DataLayerEditorSubsystem()
   n14 = GetDataLayerInstances(self=@n13, DataLayerAssets=Get KelpLinen_Day5_ToUnload() -> (KelpLinen_Day5_ToUnload:Array of Data Layer Asset Object References))
   > SetDataLayersIsLoadedInEditor(self=@n13, DataLayers=@n14, bIsLoadedInEditor=false, bIsFromUserChange=false)
   > SetDataLayersVisibility(self=@n13, DataLayers=@n14, bIsVisible=false)
  ]}
 ToggleUIElements {nodes [
   ToggleUIElements() {is=title K2Node_FunctionEntry}
   > SetIsEnabled(self=Get TopBox() -> (TopBox:Vertical Box Object Reference), bInIsEnabled=Get ValidLayerManager() -> (ValidLayerManager:bool))
   > SetIsEnabled(self=Get BottomBox() -> (BottomBox:Vertical Box Object Reference), bInIsEnabled=Get ValidEventScript() -> (ValidEventScript:bool))
  ]}
 UnloadAllKelpLinen {nodes [
   UnloadAllKelpLinen() {is=title K2Node_FunctionEntry}
   n1 = Get DataLayerEditorSubsystem()
   n2 = GetDataLayerInstances(self=@n1, DataLayerAssets=Get AllKelpLinenDatalayers() -> (AllKelpLinenDatalayers:Array of Data Layer Asset Object References))
   > SetDataLayersIsLoadedInEditor(self=@n1, DataLayers=@n2, bIsLoadedInEditor=false, bIsFromUserChange=false)
   > SetDataLayersVisibility(self=@n1, DataLayers=@n2, bIsVisible=false)
  ]}
 UnloadSeasons {nodes [
   n1 = UnloadSeasons!() -> (then, DataLayerData:Calendar Data Layer Data Structure) {is=title K2Node_FunctionEntry}
   n2 = Get DataLayerEditorSubsystem()
   n3 = GetDataLayerInstances(self=@n2, DataLayerAssets=Break Calendar Data Layer Data(CalendarDataLayerData=@n1).EnabledDataLayers_9_#2)
   > SetDataLayersIsLoadedInEditor(self=@n2, DataLayers=@n3, bIsLoadedInEditor=false, bIsFromUserChange=false)
   > SetDataLayersVisibility(self=@n2, DataLayers=@n3, bIsVisible=false)
  ]}
]
variables [
 AllKelpLinenDatalayers {
  type="array<DataLayerAsset>"
  default [
   /Clyde/DataLayers/Clyde_S2/DL_CLY_W4_D1_Forward
   /Clyde/DataLayers/Clyde_S2/DL_CLY_W4_D1_ForwardClient
   /Clyde/DataLayers/Clyde_S2/DL_CLY_W4_D1_Only
   /Clyde/DataLayers/Clyde_S2/DL_CLY_W4_D1_Only_Client
   /Clyde/DataLayers/Clyde_S2/DL_CLY_W4_D3_Forward
   /Clyde/DataLayers/Clyde_S2/DL_CLY_W4_D3_Only
   /Clyde/DataLayers/Clyde_S2/DL_CLY_W4_D3_Only_Client
   /Clyde/DataLayers/Clyde_S2/DL_CLY_W4_D5_Forward
   /Clyde/DataLayers/Clyde_S2/DL_CLY_W4_D5_Only_Client
  ]
 }
 BodiesOfWater {type="array<FortWaterBodyBP_C>", default []}
 CalendarWeek {type=number}
 DeloadPhaseArray {type="array<PhaseInfo>", default []}
 DisablePostProcessing {
  type=UnderwaterPostProcessSettings
  default {
   bEnabled=false
   blendRadius=100
   blendWeight=1
   priority=0
   postProcessSettings {
    ambientCubemap=null
    ambientCubemapIntensity=1
    ambientOcclusionBias=3
    ambientOcclusionFadeDistance=8000
    ambientOcclusionFadeRadius=5000
    ambientOcclusionIntensity=0.5
    ambientOcclusionMipBlend=0.6
    ambientOcclusionMipScale=1.7
    ambientOcclusionMipThreshold=0.01
    ambientOcclusionPower=2
    ambientOcclusionQuality=50
    ambientOcclusionRadius=200
    ambientOcclusionRadiusInWS=false
    ambientOcclusionStaticFraction=1
    ambientOcclusionTemporalBlendWeight=0.1
    autoExposureApplyPhysicalCameraExposure=true
    autoExposureBias=1
    autoExposureBiasBackup=0
    autoExposureBiasCurve=null
    autoExposureHighPercent=90
    autoExposureLowPercent=10
    autoExposureMaxBrightness=8
    autoExposureMeterMask=null
    autoExposureMethod=AEM_Histogram
    autoExposureMinBrightness=0.03
    autoExposureSpeedDown=1
    autoExposureSpeedUp=3
    bMegaLights=true
    bMobileHQGaussian=false
    bOverride_AmbientCubemapIntensity=false
    bOverride_AmbientCubemapTint=false
    bOverride_AmbientOcclusionBias=false
    bOverride_AmbientOcclusionFadeDistance=false
    bOverride_AmbientOcclusionFadeRadius=false
    bOverride_AmbientOcclusionIntensity=false
    bOverride_AmbientOcclusionMipBlend=false
    bOverride_AmbientOcclusionMipScale=false
    bOverride_AmbientOcclusionMipThreshold=false
    bOverride_AmbientOcclusionPower=false
    bOverride_AmbientOcclusionQuality=false
    bOverride_AmbientOcclusionRadius=false
    bOverride_AmbientOcclusionRadiusInWS=false
    bOverride_AmbientOcclusionStaticFraction=false
    bOverride_AmbientOcclusionTemporalBlendWeight=false
    bOverride_AutoExposureApplyPhysicalCameraExposure=false
    bOverride_AutoExposureBias=false
    bOverride_AutoExposureBiasBackup=false
    bOverride_AutoExposureBiasCurve=false
    bOverride_AutoExposureHighPercent=false
    bOverride_AutoExposureLowPercent=false
    bOverride_AutoExposureMaxBrightness=false
    bOverride_AutoExposureMeterMask=false
    bOverride_AutoExposureMethod=false
    bOverride_AutoExposureMinBrightness=false
    bOverride_AutoExposureSpeedDown=false
    bOverride_AutoExposureSpeedUp=false
    bOverride_Bloom1Size=false
    bOverride_Bloom1Tint=false
    bOverride_Bloom2Size=false
    bOverride_Bloom2Tint=false
    bOverride_Bloom3Size=false
    bOverride_Bloom3Tint=false
    bOverride_Bloom4Size=false
    bOverride_Bloom4Tint=false
    bOverride_Bloom5Size=false
    bOverride_Bloom5Tint=false
    bOverride_Bloom6Size=false
    bOverride_Bloom6Tint=false
    bOverride_BloomConvolutionBufferScale=false
    bOverride_BloomConvolutionCenterUV=false
    bOverride_BloomConvolutionIntensity=false
    bOverride_BloomConvolutionPreFilterMax=false
    bOverride_BloomConvolutionPreFilterMin=false
    bOverride_BloomConvolutionPreFilterMult=false
    bOverride_BloomConvolutionScatterDispersion=false
    bOverride_BloomConvolutionSize=false
    bOverride_BloomConvolutionTexture=false
    bOverride_BloomDirtMask=false
    bOverride_BloomDirtMaskIntensity=false
    bOverride_BloomDirtMaskTint=false
    bOverride_BloomGaussianIntensity=false
    bOverride_BloomIntensity=false
    bOverride_BloomMethod=false
    bOverride_BloomSizeScale=false
    bOverride_BloomThreshold=false
    bOverride_BlueCorrection=false
    bOverride_CameraISO=false
    bOverride_CameraShutterSpeed=false
    bOverride_ChromaticAberrationStartOffset=false
    bOverride_ColorContrast=false
    bOverride_ColorContrastHighlights=false
    bOverride_ColorContrastMidtones=false
    bOverride_ColorContrastShadows=false
    bOverride_ColorCorrectionHighlightsMax=false
    bOverride_ColorCorrectionHighlightsMin=false
    bOverride_ColorCorrectionShadowsMax=false
    bOverride_ColorGain=false
    bOverride_ColorGainHighlights=false
    bOverride_ColorGainMidtones=false
    bOverride_ColorGainShadows=false
    bOverride_ColorGamma=false
    bOverride_ColorGammaHighlights=false
    bOverride_ColorGammaMidtones=false
    bOverride_ColorGammaShadows=false
    bOverride_ColorGradingIntensity=false
    bOverride_ColorGradingLUT=false
    bOverride_ColorOffset=false
    bOverride_ColorOffsetHighlights=false
    bOverride_ColorOffsetMidtones=false
    bOverride_ColorOffsetShadows=false
    bOverride_ColorSaturation=false
    bOverride_ColorSaturationHighlights=false
    bOverride_ColorSaturationMidtones=false
    bOverride_ColorSaturationShadows=false
    bOverride_DepthOfFieldAspectRatioScalar=false
    bOverride_DepthOfFieldBarrelLength=false
    bOverride_DepthOfFieldBarrelRadius=false
    bOverride_DepthOfFieldBladeCount=false
    bOverride_DepthOfFieldDepthBlurAmount=false
    bOverride_DepthOfFieldDepthBlurRadius=false
    bOverride_DepthOfFieldEnabled=false
    bOverride_DepthOfFieldFarBlurSize=false
    bOverride_DepthOfFieldFarTransitionRegion=false
    bOverride_DepthOfFieldFocalDistance=false
    bOverride_DepthOfFieldFocalRegion=false
    bOverride_DepthOfFieldFstop=false
    bOverride_DepthOfFieldMatteBoxFlags=false
    bOverride_DepthOfFieldMinFstop=false
    bOverride_DepthOfFieldNearBlurSize=false
    bOverride_DepthOfFieldNearTransitionRegion=false
    bOverride_DepthOfFieldOcclusion=false
    bOverride_DepthOfFieldPetzvalBokeh=false
    bOverride_DepthOfFieldPetzvalBokehFalloff=false
    bOverride_DepthOfFieldPetzvalExclusionBoxExtents=false
    bOverride_DepthOfFieldPetzvalExclusionBoxRadius=false
    bOverride_DepthOfFieldScale=false
    bOverride_DepthOfFieldSensorWidth=false
    bOverride_DepthOfFieldSkyFocusDistance=false
    bOverride_DepthOfFieldSqueezeFactor=false
    bOverride_DepthOfFieldUseHairDepth=false
    bOverride_DepthOfFieldVignetteSize=false
    bOverride_DynamicGlobalIlluminationMethod=false
    bOverride_ExpandGamut=false
    bOverride_FilmBlackClip=false
    bOverride_FilmGrainHighlightsMax=false
    bOverride_FilmGrainHighlightsMin=false
    bOverride_FilmGrainIntensity=false
    bOverride_FilmGrainIntensityHighlights=false
    bOverride_FilmGrainIntensityMidtones=false
    bOverride_FilmGrainIntensityShadows=false
    bOverride_FilmGrainScale=false
    bOverride_FilmGrainShadowsMax=false
    bOverride_FilmGrainTexture=false
    bOverride_FilmShoulder=false
    bOverride_FilmSlope=false
    bOverride_FilmToe=false
    bOverride_FilmWhiteClip=false
    bOverride_HistogramLogMax=false
    bOverride_HistogramLogMin=false
    bOverride_IndirectLightingColor=false
    bOverride_IndirectLightingIntensity=false
    bOverride_LensFlareBokehShape=false
    bOverride_LensFlareBokehSize=false
    bOverride_LensFlareIntensity=false
    bOverride_LensFlareThreshold=false
    bOverride_LensFlareTint=false
    bOverride_LensFlareTints=false
    bOverride_LocalExposureBlurredLuminanceBlend=false
    bOverride_LocalExposureBlurredLuminanceKernelSizePercent=false
    bOverride_LocalExposureDetailStrength=false
    bOverride_LocalExposureHighlightContrastCurve=false
    bOverride_LocalExposureHighlightContrastScale=false
    bOverride_LocalExposureHighlightThreshold=false
    bOverride_LocalExposureHighlightThresholdStrength=false
    bOverride_LocalExposureMethod=false
    bOverride_LocalExposureMiddleGreyBias=false
    bOverride_LocalExposureShadowContrastCurve=false
    bOverride_LocalExposureShadowContrastScale=false
    bOverride_LocalExposureShadowThreshold=false
    bOverride_LocalExposureShadowThresholdStrength=false
    bOverride_LumenAmbientOcclusionIntensity=false
    bOverride_LumenDiffuseColorBoost=false
    bOverride_LumenFinalGatherLightingUpdateSpeed=false
    bOverride_LumenFinalGatherQuality=false
    bOverride_LumenFinalGatherScreenTraces=false
    bOverride_LumenFrontLayerTranslucencyReflections=false
    bOverride_LumenFullSkylightLeakingDistance=false
    bOverride_LumenMaxReflectionBounces=false
    bOverride_LumenMaxRefractionBounces=false
    bOverride_LumenMaxRoughnessToTraceReflections=false
    bOverride_LumenMaxTraceDistance=false
    bOverride_LumenRayLightingMode=false
    bOverride_LumenReflectionQuality=false
    bOverride_LumenReflectionsScreenTraces=false
    bOverride_LumenSceneDetail=false
    bOverride_LumenSceneLightingQuality=false
    bOverride_LumenSceneLightingUpdateSpeed=false
    bOverride_LumenSceneViewDistance=false
    bOverride_LumenSkylightLeaking=false
    bOverride_LumenSkylightLeakingTint=false
    bOverride_LumenSurfaceCacheResolution=false
    bOverride_MegaLightsFrontLayerTranslucency=false
    bOverride_MobileHQGaussian=false
    bOverride_MotionBlurAmount=false
    bOverride_MotionBlurMax=false
    bOverride_MotionBlurPerObjectSize=false
    bOverride_MotionBlurTargetFPS=false
    bOverride_PathTracingEnableDenoiser=false
    bOverride_PathTracingEnableEmissiveMaterials=false
    bOverride_PathTracingEnableReferenceAtmosphere=false
    bOverride_PathTracingEnableReferenceDOF=false
    bOverride_PathTracingIncludeDiffuse=false
    bOverride_PathTracingIncludeEmissive=false
    bOverride_PathTracingIncludeIndirectDiffuse=false
    bOverride_PathTracingIncludeIndirectSpecular=false
    bOverride_PathTracingIncludeIndirectVolume=false
    bOverride_PathTracingIncludeSpecular=false
    bOverride_PathTracingIncludeVolume=false
    bOverride_PathTracingMaxBounces=false
    bOverride_PathTracingMaxPathIntensity=false
    bOverride_PathTracingSamplesPerPixel=false
    bOverride_RayTracingAO=false
    bOverride_RayTracingAOIntensity=false
    bOverride_RayTracingAORadius=false
    bOverride_RayTracingAOSamplesPerPixel=false
    bOverride_RayTracingGI=false
    bOverride_RayTracingGIMaxBounces=false
    bOverride_RayTracingGISamplesPerPixel=false
    bOverride_RayTracingReflectionsMaxBounces=false
    bOverride_RayTracingReflectionsMaxRoughness=false
    bOverride_RayTracingReflectionsSamplesPerPixel=false
    bOverride_RayTracingReflectionsShadows=false
    bOverride_RayTracingReflectionsTranslucency=false
    bOverride_RayTracingTranslucencyMaxPrimaryHitEvents=false
    bOverride_RayTracingTranslucencyMaxRoughness=false
    bOverride_RayTracingTranslucencyMaxSecondaryHitEvents=false
    bOverride_RayTracingTranslucencyRefraction=false
    bOverride_RayTracingTranslucencyRefractionRays=false
    bOverride_RayTracingTranslucencySamplesPerPixel=false
    bOverride_RayTracingTranslucencyShadows=false
    bOverride_RayTracingTranslucencyUseRayTracedRefraction=false
    bOverride_ReflectionMethod=false
    bOverride_SceneColorTint=false
    bOverride_SceneFringeIntensity=false
    bOverride_ScreenSpaceReflectionIntensity=false
    bOverride_ScreenSpaceReflectionMaxRoughness=false
    bOverride_ScreenSpaceReflectionQuality=false
    bOverride_ScreenSpaceReflectionRoughnessScale=false
    bOverride_Sharpen=false
    bOverride_TemperatureType=false
    bOverride_ToneCurveAmount=false
    bOverride_TonemappingMethod=false
    bOverride_TranslucencyType=false
    bOverride_UserFlags=false
    bOverride_VignetteCenter=false
    bOverride_VignetteColor=false
    bOverride_VignetteIntensity=false
    bOverride_VignetteSize=false
    bOverride_VignetteSoftness=false
    bOverride_VignetteTexture=false
    bOverride_VignetteType=false
    bOverride_WhiteTemp=false
    bOverride_WhiteTint=false
    bOverride_bMegaLights=false
    bloom1Size=0.3
    bloom2Size=1
    bloom3Size=2
    bloom4Size=10
    bloom5Size=30
    bloom6Size=64
    bloomConvolutionBufferScale=0.133
    bloomConvolutionIntensity=1
    bloomConvolutionPreFilterMax=15000
    bloomConvolutionPreFilterMin=7
    bloomConvolutionPreFilterMult=15
    bloomConvolutionScatterDispersion=1
    bloomConvolutionSize=1
    bloomConvolutionTexture=null
    bloomDirtMask=null
    bloomDirtMaskIntensity=0
    bloomGaussianIntensity=1
    bloomIntensity=0.675
    bloomMethod=BM_SOG
    bloomSizeScale=4
    bloomThreshold=-1
    blueCorrection=0.6
    cameraISO=100
    cameraShutterSpeed=60
    chromaticAberrationStartOffset=0
    colorCorrectionHighlightsMax=1
    colorCorrectionHighlightsMin=0.5
    colorCorrectionShadowsMax=0.09
    colorGradingIntensity=1
    colorGradingLUT=null
    depthOfFieldAspectRatioScalar=1
    depthOfFieldBarrelLength=0
    depthOfFieldBarrelRadius=5
    depthOfFieldBladeCount=5
    depthOfFieldDepthBlurAmount=1
    depthOfFieldDepthBlurRadius=0
    depthOfFieldEnabled=true
    depthOfFieldFarBlurSize=15
    depthOfFieldFarTransitionRegion=500
    depthOfFieldFocalDistance=0
    depthOfFieldFocalRegion=0
    depthOfFieldFstop=4
    depthOfFieldMinFstop=1.2
    depthOfFieldNearBlurSize=15
    depthOfFieldNearTransitionRegion=300
    depthOfFieldOcclusion=0.4
    depthOfFieldPetzvalBokeh=0
    depthOfFieldPetzvalBokehFalloff=1
    depthOfFieldPetzvalExclusionBoxRadius=0
    depthOfFieldScale=0
    depthOfFieldSensorWidth=24.576
    depthOfFieldSkyFocusDistance=0
    depthOfFieldSqueezeFactor=1
    depthOfFieldUseHairDepth=false
    depthOfFieldVignetteSize=200
    dynamicGlobalIlluminationMethod=Lumen
    expandGamut=1
    filmBlackClip=0
    filmGrainHighlightsMax=1
    filmGrainHighlightsMin=0.5
    filmGrainIntensity=0
    filmGrainIntensityHighlights=1
    filmGrainIntensityMidtones=1
    filmGrainIntensityShadows=1
    filmGrainShadowsMax=0.09
    filmGrainTexture=null
    filmShoulder=0.26
    filmSlope=0.88
    filmToe=0.55
    filmWhiteClip=0.04
    histogramLogMax=4
    histogramLogMin=-8
    indirectLightingIntensity=1
    lensFlareBokehShape=null
    lensFlareBokehSize=3
    lensFlareIntensity=1
    lensFlareThreshold=8
    localExposureBlurredLuminanceBlend=0.6
    localExposureBlurredLuminanceKernelSizePercent=50
    localExposureDetailStrength=1
    localExposureHighlightContrastCurve=null
    localExposureHighlightContrastScale=1
    localExposureHighlightThreshold=0
    localExposureHighlightThresholdStrength=1
    localExposureMethod=Bilateral
    localExposureMiddleGreyBias=0
    localExposureShadowContrastCurve=null
    localExposureShadowContrastScale=1
    localExposureShadowThreshold=0
    localExposureShadowThresholdStrength=1
    lumenAmbientOcclusionIntensity=1
    lumenDiffuseColorBoost=1
    lumenFinalGatherLightingUpdateSpeed=1
    lumenFinalGatherQuality=1
    lumenFinalGatherScreenTraces=true
    lumenFrontLayerTranslucencyReflections=false
    lumenFullSkylightLeakingDistance=1000
    lumenMaxReflectionBounces=1
    lumenMaxRefractionBounces=0
    lumenMaxRoughnessToTraceReflections=0.4
    lumenMaxTraceDistance=20000
    lumenRayLightingMode=Default
    lumenReflectionQuality=1
    lumenReflectionsScreenTraces=true
    lumenSceneDetail=1
    lumenSceneLightingQuality=1
    lumenSceneLightingUpdateSpeed=1
    lumenSceneViewDistance=20000
    lumenSkylightLeaking=0
    lumenSurfaceCacheResolution=1
    megaLightsFrontLayerTranslucency=false
    motionBlurAmount=0.5
    motionBlurMax=5
    motionBlurPerObjectSize=0
    motionBlurTargetFPS=30
    pathTracingEnableDenoiser=true
    pathTracingEnableEmissiveMaterials=true
    pathTracingEnableReferenceAtmosphere=false
    pathTracingEnableReferenceDOF=false
    pathTracingIncludeDiffuse=true
    pathTracingIncludeEmissive=true
    pathTracingIncludeIndirectDiffuse=true
    pathTracingIncludeIndirectSpecular=true
    pathTracingIncludeIndirectVolume=true
    pathTracingIncludeSpecular=true
    pathTracingIncludeVolume=true
    pathTracingMaxBounces=32
    pathTracingMaxPathIntensity=24
    pathTracingSamplesPerPixel=2048
    rayTracingAO=false
    rayTracingAOIntensity=1
    rayTracingAORadius=200
    rayTracingAOSamplesPerPixel=1
    rayTracingTranslucencyMaxPrimaryHitEvents=4
    rayTracingTranslucencyMaxRoughness=0.6
    rayTracingTranslucencyMaxSecondaryHitEvents=2
    rayTracingTranslucencyRefraction=true
    rayTracingTranslucencyRefractionRays=3
    rayTracingTranslucencySamplesPerPixel=1
    rayTracingTranslucencyShadows=Hard_shadows
    rayTracingTranslucencyUseRayTracedRefraction=false
    reflectionMethod=Lumen
    sceneFringeIntensity=0
    screenSpaceReflectionIntensity=100
    screenSpaceReflectionMaxRoughness=0.6
    screenSpaceReflectionQuality=50
    sharpen=0
    temperatureType=TEMP_WhiteBalance
    toneCurveAmount=1
    tonemappingMethod=Filmic
    translucencyType=Raster
    userFlags=0
    vignetteIntensity=0.4
    vignetteSoftness=5
    vignetteTexture=null
    vignetteType=CosineFourthLaw
    whiteTemp=6500
    whiteTint=0
    ambientCubemapTint=<rgba 1 1 1 1>
    bloom1Tint=<rgba 0.3465 0.3465 0.3465 1>
    bloom2Tint=<rgba 0.138 0.138 0.138 1>
    bloom3Tint=<rgba 0.1176 0.1176 0.1176 1>
    bloom4Tint=<rgba 0.066 0.066 0.066 1>
    bloom5Tint=<rgba 0.066 0.066 0.066 1>
    bloom6Tint=<rgba 0.061 0.061 0.061 1>
    bloomConvolutionCenterUV=<0.5 0.5>
    bloomDirtMaskTint=<rgba 0.5 0.5 0.5 1>
    colorContrast=<1 1 1 1>
    colorContrastHighlights=<1 1 1 1>
    colorContrastMidtones=<1 1 1 1>
    colorContrastShadows=<1 1 1 1>
    colorGain=<1 1 1 1>
    colorGainHighlights=<1 1 1 1>
    colorGainMidtones=<1 1 1 1>
    colorGainShadows=<1 1 1 1>
    colorGamma=<1 1 1 1>
    colorGammaHighlights=<1 1 1 1>
    colorGammaMidtones=<1 1 1 1>
    colorGammaShadows=<1 1 1 1>
    colorOffset=<0 0 0 0>
    colorOffsetHighlights=<0 0 0 0>
    colorOffsetMidtones=<0 0 0 0>
    colorOffsetShadows=<0 0 0 0>
    colorSaturation=<1 1 1 1>
    colorSaturationHighlights=<1 1 1 1>
    colorSaturationMidtones=<1 1 1 1>
    colorSaturationShadows=<1 1 1 1>
    depthOfFieldMatteBoxFlags [{length=0, pitch=0, roll=0}, {length=0, pitch=0, roll=0}, {length=0, pitch=0, roll=0}]
    depthOfFieldPetzvalExclusionBoxExtents=<0 0>
    filmGrainScale=<1 1>
    indirectLightingColor=<rgba 1 1 1 1>
    lensFlareTint=<rgba 1 1 1 1>
    lensFlareTints [
     <rgba 1 0.8 0.4 0.6>
     <rgba 1 1 0.6 0.53>
     <rgba 0.8 0.8 1 0.46>
     <rgba 0.5 1 0.4 0.39>
     <rgba 0.5 0.8 1 0.31>
     <rgba 0.9 1 0.8 0.27>
     <rgba 1 0.8 0.4 0.22>
     <rgba 0.9 0.7 0.7 0.15>
    ]
    lumenSkylightLeakingTint=<rgba 1 1 1 1>
    sceneColorTint=<rgba 1 1 1 1>
    vignetteCenter=<0.5 0.5>
    vignetteColor=<rgba 0 0 0 1>
    vignetteSize=<1 1>
    weightedBlendables {array []}
   }
  }
 }
 EnablePostProcessing {
  type=UnderwaterPostProcessSettings
  default {
   bEnabled=true
   blendRadius=100
   blendWeight=1
   priority=0
   postProcessSettings {
    ambientCubemap=null
    ambientCubemapIntensity=1
    ambientOcclusionBias=3
    ambientOcclusionFadeDistance=8000
    ambientOcclusionFadeRadius=5000
    ambientOcclusionIntensity=0.5
    ambientOcclusionMipBlend=0.6
    ambientOcclusionMipScale=1.7
    ambientOcclusionMipThreshold=0.01
    ambientOcclusionPower=2
    ambientOcclusionQuality=50
    ambientOcclusionRadius=200
    ambientOcclusionRadiusInWS=false
    ambientOcclusionStaticFraction=1
    ambientOcclusionTemporalBlendWeight=0.1
    autoExposureApplyPhysicalCameraExposure=true
    autoExposureBias=1
    autoExposureBiasBackup=0
    autoExposureBiasCurve=null
    autoExposureHighPercent=90
    autoExposureLowPercent=10
    autoExposureMaxBrightness=8
    autoExposureMeterMask=null
    autoExposureMethod=AEM_Histogram
    autoExposureMinBrightness=0.03
    autoExposureSpeedDown=1
    autoExposureSpeedUp=3
    bMegaLights=true
    bMobileHQGaussian=false
    bOverride_AmbientCubemapIntensity=false
    bOverride_AmbientCubemapTint=false
    bOverride_AmbientOcclusionBias=false
    bOverride_AmbientOcclusionFadeDistance=false
    bOverride_AmbientOcclusionFadeRadius=false
    bOverride_AmbientOcclusionIntensity=false
    bOverride_AmbientOcclusionMipBlend=false
    bOverride_AmbientOcclusionMipScale=false
    bOverride_AmbientOcclusionMipThreshold=false
    bOverride_AmbientOcclusionPower=false
    bOverride_AmbientOcclusionQuality=false
    bOverride_AmbientOcclusionRadius=false
    bOverride_AmbientOcclusionRadiusInWS=false
    bOverride_AmbientOcclusionStaticFraction=false
    bOverride_AmbientOcclusionTemporalBlendWeight=false
    bOverride_AutoExposureApplyPhysicalCameraExposure=false
    bOverride_AutoExposureBias=false
    bOverride_AutoExposureBiasBackup=false
    bOverride_AutoExposureBiasCurve=false
    bOverride_AutoExposureHighPercent=false
    bOverride_AutoExposureLowPercent=false
    bOverride_AutoExposureMaxBrightness=false
    bOverride_AutoExposureMeterMask=false
    bOverride_AutoExposureMethod=false
    bOverride_AutoExposureMinBrightness=false
    bOverride_AutoExposureSpeedDown=false
    bOverride_AutoExposureSpeedUp=false
    bOverride_Bloom1Size=false
    bOverride_Bloom1Tint=false
    bOverride_Bloom2Size=false
    bOverride_Bloom2Tint=false
    bOverride_Bloom3Size=false
    bOverride_Bloom3Tint=false
    bOverride_Bloom4Size=false
    bOverride_Bloom4Tint=false
    bOverride_Bloom5Size=false
    bOverride_Bloom5Tint=false
    bOverride_Bloom6Size=false
    bOverride_Bloom6Tint=false
    bOverride_BloomConvolutionBufferScale=false
    bOverride_BloomConvolutionCenterUV=false
    bOverride_BloomConvolutionIntensity=false
    bOverride_BloomConvolutionPreFilterMax=false
    bOverride_BloomConvolutionPreFilterMin=false
    bOverride_BloomConvolutionPreFilterMult=false
    bOverride_BloomConvolutionScatterDispersion=false
    bOverride_BloomConvolutionSize=false
    bOverride_BloomConvolutionTexture=false
    bOverride_BloomDirtMask=false
    bOverride_BloomDirtMaskIntensity=false
    bOverride_BloomDirtMaskTint=false
    bOverride_BloomGaussianIntensity=false
    bOverride_BloomIntensity=false
    bOverride_BloomMethod=false
    bOverride_BloomSizeScale=false
    bOverride_BloomThreshold=false
    bOverride_BlueCorrection=false
    bOverride_CameraISO=false
    bOverride_CameraShutterSpeed=false
    bOverride_ChromaticAberrationStartOffset=false
    bOverride_ColorContrast=false
    bOverride_ColorContrastHighlights=false
    bOverride_ColorContrastMidtones=false
    bOverride_ColorContrastShadows=false
    bOverride_ColorCorrectionHighlightsMax=false
    bOverride_ColorCorrectionHighlightsMin=false
    bOverride_ColorCorrectionShadowsMax=false
    bOverride_ColorGain=false
    bOverride_ColorGainHighlights=false
    bOverride_ColorGainMidtones=false
    bOverride_ColorGainShadows=false
    bOverride_ColorGamma=false
    bOverride_ColorGammaHighlights=false
    bOverride_ColorGammaMidtones=false
    bOverride_ColorGammaShadows=false
    bOverride_ColorGradingIntensity=false
    bOverride_ColorGradingLUT=false
    bOverride_ColorOffset=false
    bOverride_ColorOffsetHighlights=false
    bOverride_ColorOffsetMidtones=false
    bOverride_ColorOffsetShadows=false
    bOverride_ColorSaturation=false
    bOverride_ColorSaturationHighlights=false
    bOverride_ColorSaturationMidtones=false
    bOverride_ColorSaturationShadows=false
    bOverride_DepthOfFieldAspectRatioScalar=false
    bOverride_DepthOfFieldBarrelLength=false
    bOverride_DepthOfFieldBarrelRadius=false
    bOverride_DepthOfFieldBladeCount=false
    bOverride_DepthOfFieldDepthBlurAmount=false
    bOverride_DepthOfFieldDepthBlurRadius=false
    bOverride_DepthOfFieldEnabled=false
    bOverride_DepthOfFieldFarBlurSize=false
    bOverride_DepthOfFieldFarTransitionRegion=false
    bOverride_DepthOfFieldFocalDistance=false
    bOverride_DepthOfFieldFocalRegion=false
    bOverride_DepthOfFieldFstop=false
    bOverride_DepthOfFieldMatteBoxFlags=false
    bOverride_DepthOfFieldMinFstop=false
    bOverride_DepthOfFieldNearBlurSize=false
    bOverride_DepthOfFieldNearTransitionRegion=false
    bOverride_DepthOfFieldOcclusion=false
    bOverride_DepthOfFieldPetzvalBokeh=false
    bOverride_DepthOfFieldPetzvalBokehFalloff=false
    bOverride_DepthOfFieldPetzvalExclusionBoxExtents=false
    bOverride_DepthOfFieldPetzvalExclusionBoxRadius=false
    bOverride_DepthOfFieldScale=false
    bOverride_DepthOfFieldSensorWidth=false
    bOverride_DepthOfFieldSkyFocusDistance=false
    bOverride_DepthOfFieldSqueezeFactor=false
    bOverride_DepthOfFieldUseHairDepth=false
    bOverride_DepthOfFieldVignetteSize=false
    bOverride_DynamicGlobalIlluminationMethod=false
    bOverride_ExpandGamut=false
    bOverride_FilmBlackClip=false
    bOverride_FilmGrainHighlightsMax=false
    bOverride_FilmGrainHighlightsMin=false
    bOverride_FilmGrainIntensity=false
    bOverride_FilmGrainIntensityHighlights=false
    bOverride_FilmGrainIntensityMidtones=false
    bOverride_FilmGrainIntensityShadows=false
    bOverride_FilmGrainScale=false
    bOverride_FilmGrainShadowsMax=false
    bOverride_FilmGrainTexture=false
    bOverride_FilmShoulder=false
    bOverride_FilmSlope=false
    bOverride_FilmToe=false
    bOverride_FilmWhiteClip=false
    bOverride_HistogramLogMax=false
    bOverride_HistogramLogMin=false
    bOverride_IndirectLightingColor=false
    bOverride_IndirectLightingIntensity=false
    bOverride_LensFlareBokehShape=false
    bOverride_LensFlareBokehSize=false
    bOverride_LensFlareIntensity=false
    bOverride_LensFlareThreshold=false
    bOverride_LensFlareTint=false
    bOverride_LensFlareTints=false
    bOverride_LocalExposureBlurredLuminanceBlend=false
    bOverride_LocalExposureBlurredLuminanceKernelSizePercent=false
    bOverride_LocalExposureDetailStrength=false
    bOverride_LocalExposureHighlightContrastCurve=false
    bOverride_LocalExposureHighlightContrastScale=false
    bOverride_LocalExposureHighlightThreshold=false
    bOverride_LocalExposureHighlightThresholdStrength=false
    bOverride_LocalExposureMethod=false
    bOverride_LocalExposureMiddleGreyBias=false
    bOverride_LocalExposureShadowContrastCurve=false
    bOverride_LocalExposureShadowContrastScale=false
    bOverride_LocalExposureShadowThreshold=false
    bOverride_LocalExposureShadowThresholdStrength=false
    bOverride_LumenAmbientOcclusionIntensity=false
    bOverride_LumenDiffuseColorBoost=false
    bOverride_LumenFinalGatherLightingUpdateSpeed=false
    bOverride_LumenFinalGatherQuality=false
    bOverride_LumenFinalGatherScreenTraces=false
    bOverride_LumenFrontLayerTranslucencyReflections=false
    bOverride_LumenFullSkylightLeakingDistance=false
    bOverride_LumenMaxReflectionBounces=false
    bOverride_LumenMaxRefractionBounces=false
    bOverride_LumenMaxRoughnessToTraceReflections=false
    bOverride_LumenMaxTraceDistance=false
    bOverride_LumenRayLightingMode=false
    bOverride_LumenReflectionQuality=false
    bOverride_LumenReflectionsScreenTraces=false
    bOverride_LumenSceneDetail=false
    bOverride_LumenSceneLightingQuality=false
    bOverride_LumenSceneLightingUpdateSpeed=false
    bOverride_LumenSceneViewDistance=false
    bOverride_LumenSkylightLeaking=false
    bOverride_LumenSkylightLeakingTint=false
    bOverride_LumenSurfaceCacheResolution=false
    bOverride_MegaLightsFrontLayerTranslucency=false
    bOverride_MobileHQGaussian=false
    bOverride_MotionBlurAmount=false
    bOverride_MotionBlurMax=false
    bOverride_MotionBlurPerObjectSize=false
    bOverride_MotionBlurTargetFPS=false
    bOverride_PathTracingEnableDenoiser=false
    bOverride_PathTracingEnableEmissiveMaterials=false
    bOverride_PathTracingEnableReferenceAtmosphere=false
    bOverride_PathTracingEnableReferenceDOF=false
    bOverride_PathTracingIncludeDiffuse=false
    bOverride_PathTracingIncludeEmissive=false
    bOverride_PathTracingIncludeIndirectDiffuse=false
    bOverride_PathTracingIncludeIndirectSpecular=false
    bOverride_PathTracingIncludeIndirectVolume=false
    bOverride_PathTracingIncludeSpecular=false
    bOverride_PathTracingIncludeVolume=false
    bOverride_PathTracingMaxBounces=false
    bOverride_PathTracingMaxPathIntensity=false
    bOverride_PathTracingSamplesPerPixel=false
    bOverride_RayTracingAO=false
    bOverride_RayTracingAOIntensity=false
    bOverride_RayTracingAORadius=false
    bOverride_RayTracingAOSamplesPerPixel=false
    bOverride_RayTracingGI=false
    bOverride_RayTracingGIMaxBounces=false
    bOverride_RayTracingGISamplesPerPixel=false
    bOverride_RayTracingReflectionsMaxBounces=false
    bOverride_RayTracingReflectionsMaxRoughness=false
    bOverride_RayTracingReflectionsSamplesPerPixel=false
    bOverride_RayTracingReflectionsShadows=false
    bOverride_RayTracingReflectionsTranslucency=false
    bOverride_RayTracingTranslucencyMaxPrimaryHitEvents=false
    bOverride_RayTracingTranslucencyMaxRoughness=false
    bOverride_RayTracingTranslucencyMaxSecondaryHitEvents=false
    bOverride_RayTracingTranslucencyRefraction=false
    bOverride_RayTracingTranslucencyRefractionRays=false
    bOverride_RayTracingTranslucencySamplesPerPixel=false
    bOverride_RayTracingTranslucencyShadows=false
    bOverride_RayTracingTranslucencyUseRayTracedRefraction=false
    bOverride_ReflectionMethod=false
    bOverride_SceneColorTint=false
    bOverride_SceneFringeIntensity=false
    bOverride_ScreenSpaceReflectionIntensity=false
    bOverride_ScreenSpaceReflectionMaxRoughness=false
    bOverride_ScreenSpaceReflectionQuality=false
    bOverride_ScreenSpaceReflectionRoughnessScale=false
    bOverride_Sharpen=false
    bOverride_TemperatureType=false
    bOverride_ToneCurveAmount=false
    bOverride_TonemappingMethod=false
    bOverride_TranslucencyType=false
    bOverride_UserFlags=false
    bOverride_VignetteCenter=false
    bOverride_VignetteColor=false
    bOverride_VignetteIntensity=false
    bOverride_VignetteSize=false
    bOverride_VignetteSoftness=false
    bOverride_VignetteTexture=false
    bOverride_VignetteType=false
    bOverride_WhiteTemp=false
    bOverride_WhiteTint=false
    bOverride_bMegaLights=false
    bloom1Size=0.3
    bloom2Size=1
    bloom3Size=2
    bloom4Size=10
    bloom5Size=30
    bloom6Size=64
    bloomConvolutionBufferScale=0.133
    bloomConvolutionIntensity=1
    bloomConvolutionPreFilterMax=15000
    bloomConvolutionPreFilterMin=7
    bloomConvolutionPreFilterMult=15
    bloomConvolutionScatterDispersion=1
    bloomConvolutionSize=1
    bloomConvolutionTexture=null
    bloomDirtMask=null
    bloomDirtMaskIntensity=0
    bloomGaussianIntensity=1
    bloomIntensity=0.675
    bloomMethod=BM_SOG
    bloomSizeScale=4
    bloomThreshold=-1
    blueCorrection=0.6
    cameraISO=100
    cameraShutterSpeed=60
    chromaticAberrationStartOffset=0
    colorCorrectionHighlightsMax=1
    colorCorrectionHighlightsMin=0.5
    colorCorrectionShadowsMax=0.09
    colorGradingIntensity=1
    colorGradingLUT=null
    depthOfFieldAspectRatioScalar=1
    depthOfFieldBarrelLength=0
    depthOfFieldBarrelRadius=5
    depthOfFieldBladeCount=5
    depthOfFieldDepthBlurAmount=1
    depthOfFieldDepthBlurRadius=0
    depthOfFieldEnabled=true
    depthOfFieldFarBlurSize=15
    depthOfFieldFarTransitionRegion=500
    depthOfFieldFocalDistance=0
    depthOfFieldFocalRegion=0
    depthOfFieldFstop=4
    depthOfFieldMinFstop=1.2
    depthOfFieldNearBlurSize=15
    depthOfFieldNearTransitionRegion=300
    depthOfFieldOcclusion=0.4
    depthOfFieldPetzvalBokeh=0
    depthOfFieldPetzvalBokehFalloff=1
    depthOfFieldPetzvalExclusionBoxRadius=0
    depthOfFieldScale=0
    depthOfFieldSensorWidth=24.576
    depthOfFieldSkyFocusDistance=0
    depthOfFieldSqueezeFactor=1
    depthOfFieldUseHairDepth=false
    depthOfFieldVignetteSize=200
    dynamicGlobalIlluminationMethod=Lumen
    expandGamut=1
    filmBlackClip=0
    filmGrainHighlightsMax=1
    filmGrainHighlightsMin=0.5
    filmGrainIntensity=0
    filmGrainIntensityHighlights=1
    filmGrainIntensityMidtones=1
    filmGrainIntensityShadows=1
    filmGrainShadowsMax=0.09
    filmGrainTexture=null
    filmShoulder=0.26
    filmSlope=0.88
    filmToe=0.55
    filmWhiteClip=0.04
    histogramLogMax=4
    histogramLogMin=-8
    indirectLightingIntensity=1
    lensFlareBokehShape=null
    lensFlareBokehSize=3
    lensFlareIntensity=1
    lensFlareThreshold=8
    localExposureBlurredLuminanceBlend=0.6
    localExposureBlurredLuminanceKernelSizePercent=50
    localExposureDetailStrength=1
    localExposureHighlightContrastCurve=null
    localExposureHighlightContrastScale=1
    localExposureHighlightThreshold=0
    localExposureHighlightThresholdStrength=1
    localExposureMethod=Bilateral
    localExposureMiddleGreyBias=0
    localExposureShadowContrastCurve=null
    localExposureShadowContrastScale=1
    localExposureShadowThreshold=0
    localExposureShadowThresholdStrength=1
    lumenAmbientOcclusionIntensity=1
    lumenDiffuseColorBoost=1
    lumenFinalGatherLightingUpdateSpeed=1
    lumenFinalGatherQuality=1
    lumenFinalGatherScreenTraces=true
    lumenFrontLayerTranslucencyReflections=false
    lumenFullSkylightLeakingDistance=1000
    lumenMaxReflectionBounces=1
    lumenMaxRefractionBounces=0
    lumenMaxRoughnessToTraceReflections=0.4
    lumenMaxTraceDistance=20000
    lumenRayLightingMode=Default
    lumenReflectionQuality=1
    lumenReflectionsScreenTraces=true
    lumenSceneDetail=1
    lumenSceneLightingQuality=1
    lumenSceneLightingUpdateSpeed=1
    lumenSceneViewDistance=20000
    lumenSkylightLeaking=0
    lumenSurfaceCacheResolution=1
    megaLightsFrontLayerTranslucency=false
    motionBlurAmount=0.5
    motionBlurMax=5
    motionBlurPerObjectSize=0
    motionBlurTargetFPS=30
    pathTracingEnableDenoiser=true
    pathTracingEnableEmissiveMaterials=true
    pathTracingEnableReferenceAtmosphere=false
    pathTracingEnableReferenceDOF=false
    pathTracingIncludeDiffuse=true
    pathTracingIncludeEmissive=true
    pathTracingIncludeIndirectDiffuse=true
    pathTracingIncludeIndirectSpecular=true
    pathTracingIncludeIndirectVolume=true
    pathTracingIncludeSpecular=true
    pathTracingIncludeVolume=true
    pathTracingMaxBounces=32
    pathTracingMaxPathIntensity=24
    pathTracingSamplesPerPixel=2048
    rayTracingAO=false
    rayTracingAOIntensity=1
    rayTracingAORadius=200
    rayTracingAOSamplesPerPixel=1
    rayTracingTranslucencyMaxPrimaryHitEvents=4
    rayTracingTranslucencyMaxRoughness=0.6
    rayTracingTranslucencyMaxSecondaryHitEvents=2
    rayTracingTranslucencyRefraction=true
    rayTracingTranslucencyRefractionRays=3
    rayTracingTranslucencySamplesPerPixel=1
    rayTracingTranslucencyShadows=Hard_shadows
    rayTracingTranslucencyUseRayTracedRefraction=false
    reflectionMethod=Lumen
    sceneFringeIntensity=0
    screenSpaceReflectionIntensity=100
    screenSpaceReflectionMaxRoughness=0.6
    screenSpaceReflectionQuality=50
    sharpen=0
    temperatureType=TEMP_WhiteBalance
    toneCurveAmount=1
    tonemappingMethod=Filmic
    translucencyType=Raster
    userFlags=0
    vignetteIntensity=0.4
    vignetteSoftness=5
    vignetteTexture=null
    vignetteType=CosineFourthLaw
    whiteTemp=6500
    whiteTint=0
    ambientCubemapTint=<rgba 1 1 1 1>
    bloom1Tint=<rgba 0.3465 0.3465 0.3465 1>
    bloom2Tint=<rgba 0.138 0.138 0.138 1>
    bloom3Tint=<rgba 0.1176 0.1176 0.1176 1>
    bloom4Tint=<rgba 0.066 0.066 0.066 1>
    bloom5Tint=<rgba 0.066 0.066 0.066 1>
    bloom6Tint=<rgba 0.061 0.061 0.061 1>
    bloomConvolutionCenterUV=<0.5 0.5>
    bloomDirtMaskTint=<rgba 0.5 0.5 0.5 1>
    colorContrast=<1 1 1 1>
    colorContrastHighlights=<1 1 1 1>
    colorContrastMidtones=<1 1 1 1>
    colorContrastShadows=<1 1 1 1>
    colorGain=<1 1 1 1>
    colorGainHighlights=<1 1 1 1>
    colorGainMidtones=<1 1 1 1>
    colorGainShadows=<1 1 1 1>
    colorGamma=<1 1 1 1>
    colorGammaHighlights=<1 1 1 1>
    colorGammaMidtones=<1 1 1 1>
    colorGammaShadows=<1 1 1 1>
    colorOffset=<0 0 0 0>
    colorOffsetHighlights=<0 0 0 0>
    colorOffsetMidtones=<0 0 0 0>
    colorOffsetShadows=<0 0 0 0>
    colorSaturation=<1 1 1 1>
    colorSaturationHighlights=<1 1 1 1>
    colorSaturationMidtones=<1 1 1 1>
    colorSaturationShadows=<1 1 1 1>
    depthOfFieldMatteBoxFlags [{length=0, pitch=0, roll=0}, {length=0, pitch=0, roll=0}, {length=0, pitch=0, roll=0}]
    depthOfFieldPetzvalExclusionBoxExtents=<0 0>
    filmGrainScale=<1 1>
    indirectLightingColor=<rgba 1 1 1 1>
    lensFlareTint=<rgba 1 1 1 1>
    lensFlareTints [
     <rgba 1 0.8 0.4 0.6>
     <rgba 1 1 0.6 0.53>
     <rgba 0.8 0.8 1 0.46>
     <rgba 0.5 1 0.4 0.39>
     <rgba 0.5 0.8 1 0.31>
     <rgba 0.9 1 0.8 0.27>
     <rgba 1 0.8 0.4 0.22>
     <rgba 0.9 0.7 0.7 0.15>
    ]
    lumenSkylightLeakingTint=<rgba 1 1 1 1>
    sceneColorTint=<rgba 1 1 1 1>
    vignetteCenter=<0.5 0.5>
    vignetteColor=<rgba 0 0 0 1>
    vignetteSize=<1 1>
    weightedBlendables {array []}
   }
  }
 }
 KelpLinen_Day1_ToLoad {
  type="array<DataLayerAsset>"
  default [
   &DL_CLY_W4_D1_Forward
   &DL_CLY_W4_D1_Only
   &DL_CLY_W4_D1_Only_Client
   &DL_CLY_W4_D1_ForwardClient
  ]
 }
 KelpLinen_Day1_ToUnload {
  type="array<DataLayerAsset>"
  default [
   &DL_CLY_W4_D3_Forward
   &DL_CLY_W4_D3_Only
   &DL_CLY_W4_D5_Forward
   &DL_CLY_W4_D3_Only_Client
   &DL_CLY_W4_D5_Only_Client
  ]
 }
 KelpLinen_Day3_ToLoad {
  type="array<DataLayerAsset>"
  default [
   &DL_CLY_W4_D3_Forward
   &DL_CLY_W4_D1_Forward
   &DL_CLY_W4_D3_Only
   &DL_CLY_W4_D3_Only_Client
   &DL_CLY_W4_D1_ForwardClient
  ]
 }
 KelpLinen_Day3_ToUnload {
  type="array<DataLayerAsset>"
  default [
   &DL_CLY_W4_D1_Only
   &DL_CLY_W4_D5_Forward
   &DL_CLY_W4_D1_Only_Client
   &DL_CLY_W4_D5_Only_Client
  ]
 }
 KelpLinen_Day5_ToLoad {
  type="array<DataLayerAsset>"
  default [
   &DL_CLY_W4_D1_Forward
   &DL_CLY_W4_D3_Forward
   &DL_CLY_W4_D5_Forward
   &DL_CLY_W4_D5_Only_Client
   &DL_CLY_W4_D1_ForwardClient
  ]
 }
 KelpLinen_Day5_ToUnload {
  type="array<DataLayerAsset>"
  default [
   &DL_CLY_W4_D3_Only
   &DL_CLY_W4_D1_Only
   &DL_CLY_W4_D1_Only_Client
   &DL_CLY_W4_D3_Only_Client
  ]
 }
 LayerManager {type=BP_Clyde_DataLayerManager_C, default=null}
 MatchingPhaseInfo {
  type=PhaseInfo
  default {
   bDestroySpawnedLevelSequenceActorOnPhaseEnd=true
   bGarbageCollectAtPhaseStartOnClient=true
   bGarbageCollectAtPhaseStartOnServer=false
   bSetRelevancyMode=true
   bStartNextPhaseOnPhaseFinished=false
   dataLayerConfig=null
   levelSequenceActor=null
   numOfPlayersInRelevancyGroup=1
   phaseActor=null
   phaseDuration=0
   relevancyMode=NormalRelevancy
   replicatedLevelSequenceActorClass=null
   replicatedLevelSequenceActorClass_HardRef=null
   dataLayers []
   levels []
   nextPhaseTagOverride {tagName=None}
   phaseTag {tagName=None}
   requireLevelsFromPhaseTags {gameplayTags []}
  }
 }
 Phase Info Array {type="array<PhaseInfo>", default []}
 PossibleLayerManagers {type="array<BP_Clyde_DataLayerManager_C>", default []}
 SeasonText {type=number}
 Special Event Script {type=SpecialEventScript, default=null}
 SpecialEventScriptClass {type=Class, default=/Script/SpecialEventGameplayRuntime.SpecialEventScript}
 ValidEventScript {type=boolean}
 ValidLayerManager {type=boolean}
 WeekName {type=string, default=ClydeSeason}
]
widgets {
 info {rootWidgetClass=&StackBox, widgetCount=38}
 widgets [
  StackBox_0 {widgetClassPath=&StackBox, widget=StackBox_0, properties {orientation=Orient_Vertical}}
  TextBlock_67 {
   widgetClassPath=&TextBlock
   parent=StackBox_0
   widget=TextBlock_67
   properties {
    justification=Center
    text=Season and Event Loader
    font {
     bForceMonospaced=false
     bMaterialIsStencil=false
     fontMaterial=null
     fontObject=&Roboto
     letterSpacing=0
     monospacedWidth=1
     size=15
     skewAmount=0
     typefaceFontName=Bold
     outlineSettings {
      bApplyOutlineToDropShadows=false
      bMiteredCorners=false
      bSeparateFillAlpha=false
      outlineMaterial=null
      outlineSize=0
      outlineColor=<rgba 0 0 0 1>
     }
    }
   }
   slot {class=StackBoxSlot, properties {horizontalAlignment=HAlign_Center, padding=<ltrb 0 0 0 10>}}
  }
  RefreshLayer {
   widgetClassPath=&EditorUtilityButton
   parent=StackBox_0
   widget=RefreshLayer
   properties {toolTipText=Try to get a reference to a loaded Data Layer Manager.}
   slot {class=StackBoxSlot, properties {padding=<ltrb 50 0 50 5>}}
  }
  TextBlock_33 {
   widgetClassPath=&TextBlock
   parent=RefreshLayer
   widget=TextBlock_33
   properties {
    text=Get Layer Manager
    font {
     bForceMonospaced=false
     bMaterialIsStencil=false
     fontMaterial=null
     fontObject=&Roboto
     letterSpacing=0
     monospacedWidth=1
     size=10.5
     skewAmount=0
     typefaceFontName=Regular
     outlineSettings {
      bApplyOutlineToDropShadows=false
      bMiteredCorners=false
      bSeparateFillAlpha=false
      outlineMaterial=null
      outlineSize=0
      outlineColor=<rgba 0 0 0 1>
     }
    }
   }
   slot {class=ButtonSlot}
  }
  RefreshEvents {
   widgetClassPath=&EditorUtilityButton
   parent=StackBox_0
   widget=RefreshEvents
   properties {toolTipText=Try to get a reference to a loaded Special Event Script.}
   slot {class=StackBoxSlot, properties {padding=<ltrb 50 0 50 5>}}
  }
  TextBlock_8 {
   widgetClassPath=&TextBlock
   parent=RefreshEvents
   widget=TextBlock_8
   properties {
    text=Get Special Event Script
    font {
     bForceMonospaced=false
     bMaterialIsStencil=false
     fontMaterial=null
     fontObject=&Roboto
     letterSpacing=0
     monospacedWidth=1
     size=10.5
     skewAmount=0
     typefaceFontName=Regular
     outlineSettings {
      bApplyOutlineToDropShadows=false
      bMiteredCorners=false
      bSeparateFillAlpha=false
      outlineMaterial=null
      outlineSize=0
      outlineColor=<rgba 0 0 0 1>
     }
    }
   }
   slot {class=ButtonSlot}
  }
  RefreshBoW {
   widgetClassPath=&EditorUtilityButton
   parent=StackBox_0
   widget=RefreshBoW
   properties {toolTipText=Refresh the bodies of water.}
   slot {class=StackBoxSlot, properties {padding=<ltrb 50 0 50 5>}}
  }
  TextBlock_2 {
   widgetClassPath=&TextBlock
   parent=RefreshBoW
   widget=TextBlock_2
   properties {
    text=Refresh Bodies of Water
    font {
     bForceMonospaced=false
     bMaterialIsStencil=false
     fontMaterial=null
     fontObject=&Roboto
     letterSpacing=0
     monospacedWidth=1
     size=10.5
     skewAmount=0
     typefaceFontName=Regular
     outlineSettings {
      bApplyOutlineToDropShadows=false
      bMiteredCorners=false
      bSeparateFillAlpha=false
      outlineMaterial=null
      outlineSize=0
      outlineColor=<rgba 0 0 0 1>
     }
    }
   }
   slot {class=ButtonSlot}
  }
  TopBox {widgetClassPath=&VerticalBox, bIsVariable=true, parent=StackBox_0, widget=TopBox, slot {class=StackBoxSlot}}
  SelectButton {
   widgetClassPath=&EditorUtilityButton
   parent=TopBox
   widget=SelectButton
   properties {toolTipText=Select the Layer Manager we are using.}
   slot {class=VerticalBoxSlot, properties {padding=<ltrb 50 0 50 15>}}
  }
  TextBlock {
   widgetClassPath=&TextBlock
   parent=SelectButton
   widget=TextBlock
   properties {
    text=Select Manager in Outliner
    font {
     bForceMonospaced=false
     bMaterialIsStencil=false
     fontMaterial=null
     fontObject=&Roboto
     letterSpacing=0
     monospacedWidth=1
     size=10.5
     skewAmount=0
     typefaceFontName=Regular
     outlineSettings {
      bApplyOutlineToDropShadows=false
      bMiteredCorners=false
      bSeparateFillAlpha=false
      outlineMaterial=null
      outlineSize=0
      outlineColor=<rgba 0 0 0 1>
     }
    }
   }
   slot {class=ButtonSlot}
  }
  TextBlock_3 {
   widgetClassPath=&TextBlock
   parent=TopBox
   widget=TextBlock_3
   properties {
    justification=Center
    text=Seasons
    font {
     bForceMonospaced=false
     bMaterialIsStencil=false
     fontMaterial=null
     fontObject=&Roboto
     letterSpacing=0
     monospacedWidth=1
     size=15
     skewAmount=0
     typefaceFontName=Bold
     outlineSettings {
      bApplyOutlineToDropShadows=false
      bMiteredCorners=false
      bSeparateFillAlpha=false
      outlineMaterial=null
      outlineSize=0
      outlineColor=<rgba 0 0 0 1>
     }
    }
   }
   slot {class=VerticalBoxSlot, properties {horizontalAlignment=HAlign_Center, padding=<ltrb 0 0 0 10>}}
  }
  HorizontalBox_0 {
   widgetClassPath=&HorizontalBox
   parent=TopBox
   widget=HorizontalBox_0
   properties {toolTipText="The currently loaded Week. If 0, tool hasn't loaded any week."}
   slot {class=VerticalBoxSlot, properties {padding=<ltrb 0 0 0 10>}}
  }
  Calendar {
   widgetClassPath=&TextBlock
   parent=HorizontalBox_0
   widget=Calendar
   properties {
    justification=Right
    text=Current Week:
    font {
     bForceMonospaced=false
     bMaterialIsStencil=false
     fontMaterial=null
     fontObject=&Roboto
     letterSpacing=0
     monospacedWidth=1
     size=10.5
     skewAmount=0
     typefaceFontName=Regular
     outlineSettings {
      bApplyOutlineToDropShadows=false
      bMiteredCorners=false
      bSeparateFillAlpha=false
      outlineMaterial=null
      outlineSize=0
      outlineColor=<rgba 0 0 0 1>
     }
    }
   }
   slot {class=HorizontalBoxSlot, properties {horizontalAlignment=HAlign_Right, size {sizeRule=Fill, value=1}}}
  }
  Week {
   widgetClassPath=&TextBlock
   bIsVariable=true
   parent=HorizontalBox_0
   widget=Week
   properties {
    text=0
    font {
     bForceMonospaced=false
     bMaterialIsStencil=false
     fontMaterial=null
     fontObject=&Roboto
     letterSpacing=0
     monospacedWidth=1
     size=10.5
     skewAmount=0
     typefaceFontName=Regular
     outlineSettings {
      bApplyOutlineToDropShadows=false
      bMiteredCorners=false
      bSeparateFillAlpha=false
      outlineMaterial=null
      outlineSize=0
      outlineColor=<rgba 0 0 0 1>
     }
    }
   }
   slot {class=HorizontalBoxSlot, properties {horizontalAlignment=HAlign_Left, size {sizeRule=Fill, value=1}}}
  }
  HorizontalBox {
   widgetClassPath=&HorizontalBox
   parent=TopBox
   widget=HorizontalBox
   properties {toolTipText="When checked, the \"Load Season\" Buttons will also unload all of the Layers enabled by other Buttons. NOTE: If there seem to be problems loading/unloading data layers, check this off as your first test."}
   slot {class=VerticalBoxSlot, properties {padding=<ltrb 0 0 0 10>}}
  }
  Calendar_1 {
   widgetClassPath=&TextBlock
   parent=HorizontalBox
   widget=Calendar_1
   properties {
    justification=Right
    text=Unload Weeks
    font {
     bForceMonospaced=false
     bMaterialIsStencil=false
     fontMaterial=null
     fontObject=&Roboto
     letterSpacing=0
     monospacedWidth=1
     size=10.5
     skewAmount=0
     typefaceFontName=Regular
     outlineSettings {
      bApplyOutlineToDropShadows=false
      bMiteredCorners=false
      bSeparateFillAlpha=false
      outlineMaterial=null
      outlineSize=0
      outlineColor=<rgba 0 0 0 1>
     }
    }
   }
   slot {class=HorizontalBoxSlot, properties {horizontalAlignment=HAlign_Right, size {sizeRule=Fill, value=1}}}
  }
  UnloadCheckBox {
   widgetClassPath=&EditorUtilityCheckBox
   bIsVariable=true
   parent=HorizontalBox
   widget=UnloadCheckBox
   slot {class=HorizontalBoxSlot, properties {horizontalAlignment=HAlign_Left, size {sizeRule=Fill, value=1}}}
  }
  UnloadButton {
   widgetClassPath=&EditorUtilityButton
   parent=TopBox
   widget=UnloadButton
   properties {toolTipText="Unload All of the Data Layers under the \"Enabled Data Layers\" struct in the Data Layer Manager."}
   slot {class=VerticalBoxSlot, properties {padding=<ltrb 50 5 50 10>}}
  }
  TextBlock_5 {
   widgetClassPath=&TextBlock
   parent=UnloadButton
   widget=TextBlock_5
   properties {
    text=Unload All Weeks
    font {
     bForceMonospaced=false
     bMaterialIsStencil=false
     fontMaterial=null
     fontObject=&Roboto
     letterSpacing=0
     monospacedWidth=1
     size=10.5
     skewAmount=0
     typefaceFontName=Regular
     outlineSettings {
      bApplyOutlineToDropShadows=false
      bMiteredCorners=false
      bSeparateFillAlpha=false
      outlineMaterial=null
      outlineSize=0
      outlineColor=<rgba 0 0 0 1>
     }
    }
   }
   slot {class=ButtonSlot}
  }
  WeekButton_1 {
   widgetClassPath=&EditorUtilityButton
   parent=TopBox
   widget=WeekButton_1
   properties {toolTipText="Load week 1 of the current \"Week Name\". This will load and unload the proper layers based on the assigned row."}
   slot {class=VerticalBoxSlot, properties {padding=<ltrb 50 5 50 10>}}
  }
  TextBlock_302 {
   widgetClassPath=&TextBlock
   parent=WeekButton_1
   widget=TextBlock_302
   properties {
    text=Week 1
    font {
     bForceMonospaced=false
     bMaterialIsStencil=false
     fontMaterial=null
     fontObject=&Roboto
     letterSpacing=0
     monospacedWidth=1
     size=10.5
     skewAmount=0
     typefaceFontName=Regular
     outlineSettings {
      bApplyOutlineToDropShadows=false
      bMiteredCorners=false
      bSeparateFillAlpha=false
      outlineMaterial=null
      outlineSize=0
      outlineColor=<rgba 0 0 0 1>
     }
    }
   }
   slot {class=ButtonSlot}
  }
  WeekButton_2 {
   widgetClassPath=&EditorUtilityButton
   parent=TopBox
   widget=WeekButton_2
   properties {toolTipText="Load week 2 of the current \"Week Name\". This will load and unload the proper layers based on the assigned row."}
   slot {class=VerticalBoxSlot, properties {padding=<ltrb 50 5 50 10>}}
  }
  TextBlock_1 {
   widgetClassPath=&TextBlock
   parent=WeekButton_2
   widget=TextBlock_1
   properties {
    text=Week 2
    font {
     bForceMonospaced=false
     bMaterialIsStencil=false
     fontMaterial=null
     fontObject=&Roboto
     letterSpacing=0
     monospacedWidth=1
     size=10.5
     skewAmount=0
     typefaceFontName=Regular
     outlineSettings {
      bApplyOutlineToDropShadows=false
      bMiteredCorners=false
      bSeparateFillAlpha=false
      outlineMaterial=null
      outlineSize=0
      outlineColor=<rgba 0 0 0 1>
     }
    }
   }
   slot {class=ButtonSlot}
  }
  WeekButton_3 {
   widgetClassPath=&EditorUtilityButton
   parent=TopBox
   widget=WeekButton_3
   properties {toolTipText="Load week 3 of the current \"Week Name\". This will load and unload the proper layers based on the assigned row."}
   slot {class=VerticalBoxSlot, properties {padding=<ltrb 50 5 50 10>}}
  }
  TextBlock_4 {
   widgetClassPath=&TextBlock
   parent=WeekButton_3
   widget=TextBlock_4
   properties {
    text=Week 3
    font {
     bForceMonospaced=false
     bMaterialIsStencil=false
     fontMaterial=null
     fontObject=&Roboto
     letterSpacing=0
     monospacedWidth=1
     size=10.5
     skewAmount=0
     typefaceFontName=Regular
     outlineSettings {
      bApplyOutlineToDropShadows=false
      bMiteredCorners=false
      bSeparateFillAlpha=false
      outlineMaterial=null
      outlineSize=0
      outlineColor=<rgba 0 0 0 1>
     }
    }
   }
   slot {class=ButtonSlot}
  }
  WeekButton_4 {
   widgetClassPath=&EditorUtilityButton
   parent=TopBox
   widget=WeekButton_4
   properties {toolTipText="Load week 4 of the current \"Week Name\". This will load and unload the proper layers based on the assigned row."}
   slot {class=VerticalBoxSlot, properties {padding=<ltrb 50 5 50 10>}}
  }
  TextBlock_6 {
   widgetClassPath=&TextBlock
   parent=WeekButton_4
   widget=TextBlock_6
   properties {
    text=Week 4
    font {
     bForceMonospaced=false
     bMaterialIsStencil=false
     fontMaterial=null
     fontObject=&Roboto
     letterSpacing=0
     monospacedWidth=1
     size=10.5
     skewAmount=0
     typefaceFontName=Regular
     outlineSettings {
      bApplyOutlineToDropShadows=false
      bMiteredCorners=false
      bSeparateFillAlpha=false
      outlineMaterial=null
      outlineSize=0
      outlineColor=<rgba 0 0 0 1>
     }
    }
   }
   slot {class=ButtonSlot}
  }
  HorizontalBox_2 {
   widgetClassPath=&HorizontalBox
   parent=TopBox
   widget=HorizontalBox_2
   slot {class=VerticalBoxSlot, properties {horizontalAlignment=HAlign_Center, verticalAlignment=VAlign_Center}}
  }
  Calendar_2 {
   widgetClassPath=&TextBlock
   parent=HorizontalBox_2
   widget=Calendar_2
   properties {
    justification=Right
    text=Kelp Linen Days
    font {
     bForceMonospaced=false
     bMaterialIsStencil=false
     fontMaterial=null
     fontObject=&Roboto
     letterSpacing=0
     monospacedWidth=1
     size=10.5
     skewAmount=0
     typefaceFontName=Regular
     outlineSettings {
      bApplyOutlineToDropShadows=false
      bMiteredCorners=false
      bSeparateFillAlpha=false
      outlineMaterial=null
      outlineSize=0
      outlineColor=<rgba 0 0 0 1>
     }
    }
   }
   slot {class=HorizontalBoxSlot, properties {horizontalAlignment=HAlign_Center, size {sizeRule=Fill, value=1}}}
  }
  HorizontalBox_1 {
   widgetClassPath=&HorizontalBox
   parent=TopBox
   widget=HorizontalBox_1
   slot {class=VerticalBoxSlot, properties {horizontalAlignment=HAlign_Center, verticalAlignment=VAlign_Center}}
  }
  KelpButton_1 {
   widgetClassPath=&EditorUtilityButton
   parent=HorizontalBox_1
   widget=KelpButton_1
   properties {toolTipText=Loads week 4 and kelp linen day 1}
   slot {class=HorizontalBoxSlot, properties {padding=<ltrb 50 5 50 10>}}
  }
  Day1 {
   widgetClassPath=&TextBlock
   parent=KelpButton_1
   widget=Day1
   properties {
    text=Week 4 Day 1
    font {
     bForceMonospaced=false
     bMaterialIsStencil=false
     fontMaterial=null
     fontObject=&Roboto
     letterSpacing=0
     monospacedWidth=1
     size=10.5
     skewAmount=0
     typefaceFontName=Regular
     outlineSettings {
      bApplyOutlineToDropShadows=false
      bMiteredCorners=false
      bSeparateFillAlpha=false
      outlineMaterial=null
      outlineSize=0
      outlineColor=<rgba 0 0 0 1>
     }
    }
   }
   slot {class=ButtonSlot}
  }
  KelpButton_2 {
   widgetClassPath=&EditorUtilityButton
   parent=HorizontalBox_1
   widget=KelpButton_2
   properties {toolTipText=Loads week 4 and kelp linen day 3}
   slot {class=HorizontalBoxSlot, properties {padding=<ltrb 50 5 50 10>}}
  }
  Day3 {
   widgetClassPath=&TextBlock
   parent=KelpButton_2
   widget=Day3
   properties {
    text=Week 4 Day 3
    font {
     bForceMonospaced=false
     bMaterialIsStencil=false
     fontMaterial=null
     fontObject=&Roboto
     letterSpacing=0
     monospacedWidth=1
     size=10.5
     skewAmount=0
     typefaceFontName=Regular
     outlineSettings {
      bApplyOutlineToDropShadows=false
      bMiteredCorners=false
      bSeparateFillAlpha=false
      outlineMaterial=null
      outlineSize=0
      outlineColor=<rgba 0 0 0 1>
     }
    }
   }
   slot {class=ButtonSlot}
  }
  KelpButton_3 {
   widgetClassPath=&EditorUtilityButton
   parent=HorizontalBox_1
   widget=KelpButton_3
   properties {toolTipText=Loads week 4 and kelp linen day 5}
   slot {class=HorizontalBoxSlot, properties {padding=<ltrb 50 5 50 10>}}
  }
  Day5 {
   widgetClassPath=&TextBlock
   parent=KelpButton_3
   widget=Day5
   properties {
    text=Week 4 Day 5
    font {
     bForceMonospaced=false
     bMaterialIsStencil=false
     fontMaterial=null
     fontObject=&Roboto
     letterSpacing=0
     monospacedWidth=1
     size=10.5
     skewAmount=0
     typefaceFontName=Regular
     outlineSettings {
      bApplyOutlineToDropShadows=false
      bMiteredCorners=false
      bSeparateFillAlpha=false
      outlineMaterial=null
      outlineSize=0
      outlineColor=<rgba 0 0 0 1>
     }
    }
   }
   slot {class=ButtonSlot}
  }
  BottomBox {widgetClassPath=&VerticalBox, bIsVariable=true, parent=StackBox_0, widget=BottomBox, slot {class=StackBoxSlot}}
 ]
}
