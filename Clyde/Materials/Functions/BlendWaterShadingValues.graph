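# Material function that blends two sets of water shading values into one.
# Two linear interpolations share a scalar Alpha input, mixing Scattering A with B and Absorption A with B.
# The results leave through separate Scattering and Absorption outputs.
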MaterialFunction /Clyde/Materials/Functions/BlendWaterShadingValues
graph {nodes [
  n1 = FunctionInput() {bUsePreviewValueAsDefault=true, inputName=Alpha, inputType=FunctionInput_Scalar, sortPriority=4}
  FunctionOutput(None=LinearInterpolate(
    A=FunctionInput() {bUsePreviewValueAsDefault=true, inputName=Scattering A}
    B=FunctionInput() {bUsePreviewValueAsDefault=true, inputName=Scattering B, sortPriority=1}
    Alpha=@n1
   )) {outputName=Scattering}
  FunctionOutput(None=LinearInterpolate(
    A=FunctionInput() {bUsePreviewValueAsDefault=true, inputName=Absorption A, sortPriority=2}
    B=FunctionInput() {bUsePreviewValueAsDefault=true, inputName=Absorption B, sortPriority=3}
    Alpha=@n1
   )) {outputName=Absorption, sortPriority=1}
 ]}
