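# Material function for a deep Gerstner water wave that outputs WPO, normal and two wave positions.
# Wave phase comes from XY position, direction and time (water MPC by default), wrapped to 2π.
# With Z Only Offset on, it brackets the point with two wave samples, ordered by half-period, and linearly interpolates their height and normal using horizontal offsets.
# This approximates the wave surface using vertical displacement alone.
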MaterialFunction /Clyde/Materials/Functions/GerstnerWave_Deep_WithZTrace
graph {nodes [
  n1 = MaterialFunctionCall()
  n2 = AppendVector(A=@n1, B=@n1) {desc="first "}
  n3 = MaterialFunctionCall()
  n4 = MaterialFunctionCall(Multiplier=Constant() {r=2}) {materialFunction=&Pi}
  n5 = FunctionInput() {bUsePreviewValueAsDefault=true, inputName=Wave Length, inputType=FunctionInput_Scalar, sortPriority=1, previewValue=<100 0 0 1>}
  n6 = FunctionInput() {bUsePreviewValueAsDefault=true, inputName=Amplitude, inputType=FunctionInput_Scalar, sortPriority=2, previewValue=<10 0 0 1>}
  n7 = FunctionInput() {bUsePreviewValueAsDefault=true, inputName=Wave Steepness, inputType=FunctionInput_Scalar, sortPriority=4, previewValue=<0.8 0 0 1>}
  n8 = Divide(A=@n7, B=Multiply(A=@n6, B=Divide(A=@n4.Result, B=@n5) {desc=k}) {desc=wKA}) {desc=Q}
  n9 = Subtract(A=@n3, B=@n8)
  n10 = FunctionInput(Preview=WorldPosition() {worldPositionShaderOffset=WPT_ExcludeAllShaderOffsets}.XYZ) {bUsePreviewValueAsDefault=true, inputName=Pos, sortPriority=8}
  n11 = FunctionInput() {bUsePreviewValueAsDefault=true, inputName=2D Direction Vector, inputType=FunctionInput_Vector2, previewValue=<1 0 0 1>}
  n12 = FunctionInput(Preview=CollectionParameter() {collection=/Water/Materials/MPC/MPC_Water, parameterName=Time}) {bUsePreviewValueAsDefault=true, inputName=Manual Time, inputType=FunctionInput_Scalar, sortPriority=15, previewValue=<1 0 0 1>}
  n13 = Subtract(A=DotProduct(A=ComponentMask(None=@n10) {g=true, r=true}, B=@n11), B=@n12)
  n14 = MaterialFunctionCall(Multiplier=Constant() {r=2}) {materialFunction=&Pi}
  n15 = Fmod(A=@n13, B=@n14.Result)
  n16 = If(A=@n13, B=Constant(), "A > B"=@n15, "A == B"=@n15, "A < B"=Subtract(A=@n14.Result, B=Abs(None=@n15)))
  n17 = AppendVector(
   A=@n3
   B=If(
    A=@n16
    B=MaterialFunctionCall() {materialFunction=&Pi}.Result
    "A > B"=@n9
    "A == B"=@n9
    "A < B"=Add(A=@n8, B=@n3)
   ) {desc="mapped pos 1d  (new second offset 1d)"}
  ) {desc=second}
  n18 = Constant() {r=1}
  n19 = If(
   A=@n16
   B=MaterialFunctionCall() {materialFunction=&Pi}.Result
   "A > B"=@n18
   "A == B"=@n18
   "A < B"=Constant()
  ) {desc="1= flip"}
  n20 = LinearInterpolate(A=@n2, B=@n17, Alpha=@n19)
  ComponentMask(None=@n20) {b=true, desc=first offset, g=true, r=true}
  n21 = LinearInterpolate(A=@n17, B=@n2, Alpha=@n19)
  ComponentMask(None=@n21) {b=true, desc=second offset, g=true, r=true}
  FunctionInput() {bUsePreviewValueAsDefault=true, inputName=wavesin 2, inputType=FunctionInput_Scalar, sortPriority=10}
  FunctionInput() {bUsePreviewValueAsDefault=true, inputName=wavecos 2, inputType=FunctionInput_Scalar, sortPriority=11}
  FunctionInput() {bUsePreviewValueAsDefault=true, inputName=wavesin 1, inputType=FunctionInput_Scalar, sortPriority=10}
  FunctionInput() {bUsePreviewValueAsDefault=true, inputName=wavecos 1, inputType=FunctionInput_Scalar, sortPriority=11}
  FunctionOutput(None=@n3) {outputName=Wave Pos 2}
  n22 = FunctionInput(Preview=StaticBool()) {bUsePreviewValueAsDefault=true, inputName=Z Only Offset, inputType=FunctionInput_StaticBool}
  n23 = Constant2Vector()
  n24 = ComponentMask(None=@n20) {a=true, desc=First 1d offset}
  n25 = Subtract(A=ComponentMask(None=@n21) {a=true, desc=second 1d offset}, B=@n24)
  n26 = Constant() {r=1}
  n27 = Divide(A=Subtract(B=@n24) {constA=0}, B=If(A=@n25, B=Constant(), "A > B"=@n25, "A == B"=@n26, "A < B"=@n26))
  n28 = LinearInterpolate(A=ComponentMask(None=@n20) {b=true}, B=ComponentMask(None=@n21) {b=true, desc="z offset "}, Alpha=@n27)
  FunctionOutput(None=StaticSwitch(True=AppendVector(A=@n23, B=@n28), False=@n1, Value=@n22)) {outputName=WPO}
  n29 = LinearInterpolate(A=@n1, B=@n3, Alpha=@n19)
  n30 = LinearInterpolate(A=@n3, B=@n1, Alpha=@n19)
  FunctionOutput(None=StaticSwitch(True=LinearInterpolate(A=@n29, B=@n30, Alpha=@n27), False=@n1, Value=@n22)) {outputName=Normal}
  FunctionOutput(None=@n1) {outputName=Wave Pos 1}
  Reroute(None=@n10)
  Reroute(None=@n11)
  Reroute(None=@n5)
  Reroute(None=@n6)
  Reroute(None=@n7)
  n31 = Multiply(A=@n11, B=@n1)
  n32 = Multiply(A=@n31) {constB=-1}
  n33 = ComponentMask(None=If(
    A=@n16
    B=MaterialFunctionCall() {materialFunction=&Pi}.Result
    "A > B"=@n32
    "A == B"=@n32
    "A < B"=@n31
   )) {g=true, r=true}
  n34 = ComponentMask(None=@n10) {g=true, r=true}
  Reroute(None=AppendVector(A=Add(A=@n33, B=@n34), B=Constant()))
  Reroute(None=@n12)
  Reroute()
 ]}
