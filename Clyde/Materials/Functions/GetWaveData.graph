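# Material function that reads one wave's parameters from the WaveData lookup texture, by Wave Index and Water Body Index.
# Each water body maps to a row and each wave to two adjacent texels, with half-texel-centered UVs assuming a 16x64 texture.
# The first texel supplies Direction; the second supplies wavelength, amplitude and steepness, also fed into Params.
# Sampling uses an explicit mip level with automatic view mip bias off.
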MaterialFunction /Clyde/Materials/Functions/GetWaveData
graph {nodes [
  n1 = Divide(B=Constant2Vector() {g=64, r=16}) {constA=0.5}
  n2 = Multiply(A=FunctionInput(Preview=Constant()) {bUsePreviewValueAsDefault=true, inputName=Wave Index, inputType=FunctionInput_Scalar}) {constB=2}
  n3 = MaterialFunctionCall(Float2=Constant2Vector() {g=64, r=16}) {materialFunction=&BreakOutFloat2Components}
  n4 = Divide(A=Add(A=@n2), B=@n3.R)
  n5 = Divide(A=FunctionInput() {bUsePreviewValueAsDefault=true, inputName=Water Body Index, inputType=FunctionInput_Scalar, sortPriority=1}, B=@n3.G)
  n6 = TextureSampleParameter2D!(UVs=Add(A=@n1, B=AppendVector(A=@n4, B=@n5)), Level, Apply View MipBias) {
   automaticViewMipBias=false
   constMipValue=0
   group=Waves
   mipValueMode=TMVM_MipLevel
   parameterName=WaveData
   samplerSource=SSM_Wrap_WorldGroupSettings
   texture=null
  }
  FunctionOutput(None=@n6.R) {outputName=Wavlength, sortPriority=2}
  FunctionOutput(None=@n6.G) {outputName=Amplitude, sortPriority=3}
  FunctionOutput(None=@n6.B) {outputName=Steepness, sortPriority=4}
  n7 = Divide(A=@n2, B=@n3.R)
  n8 = TextureSampleParameter2D(UVs=Add(A=AppendVector(A=@n7, B=@n5), B=@n1)) {
   automaticViewMipBias=false
   constMipValue=0
   group=Waves
   mipValueMode=TMVM_MipLevel
   parameterName=WaveData
   samplerSource=SSM_Wrap_WorldGroupSettings
   texture=null
  }
  FunctionOutput(None=ComponentMask(None=@n8.RGB) {g=true, r=true}) {outputName=Direction}
  FunctionOutput(None=AppendVector(A=@n6.RGB, B=@n6.RGB)) {outputName=Params, sortPriority=1}
 ]}
