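# Material function that layers pond scum over incoming material attributes, also outputting the scum's Color, Mask and Normal.
# A world-XY swamp mask texture drives color variation and, combined with Depth, minus Sim Erasure, a contrast-adjusted coverage mask scaled by Alpha Mask.
# Coverage blends specular, roughness, opacity and a tiled moss normal, while base color blends by Alpha Mask alone.
# Forward and mobile paths only blend in a single scum color.
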MaterialFunction /Clyde/Materials/Functions/PondScumLayer
graph {nodes [
  n1 = GetMaterialAttributes(None=FunctionInput() {bUsePreviewValueAsDefault=true, inputName=Attributes, inputType=FunctionInput_MaterialAttributes}) {attributeGetTypes [#1]}
  n2 = GetMaterialAttributes(None=@n1.MaterialAttributes) {attributeGetTypes [
    #2
    #3
    #4
    #5
    #6
   ]}
  n3 = VectorParameter() {group=Foam and Pondscum, parameterName=Pondscum Emissive, defaultValue=<rgba 0.03 0.0255 0.0005 0.7>}
  n4 = Multiply(A=@n3.RGB, B=@n3.RGB)
  n5 = VectorParameter() {group=Foam and Pondscum, parameterName=Pondscum Emissive_2, defaultValue=<rgba 0.0128 0.015 0.0015 2.5>}
  n6 = ScalarParameter() {group=Foam and Pondscum, parameterName=Pondscum_ColorVariationContrast}
  n7 = ScalarParameter() {defaultValue=2, group=Foam and Pondscum, parameterName=Pondscum_ColorVariationContrast_2}
  n8 = ComponentMask(None=WorldPosition() {worldPositionShaderOffset=WPT_ExcludeAllShaderOffsets}.XYZ) {g=true, r=true}
  n9 = TextureSampleParameter2D(UVs=Divide(A=@n8, B=ScalarParameter() {defaultValue=4500, group=Foam and Pondscum, parameterName=Pondscum Tiling})) {
   group=Foam and Pondscum
   parameterName=Swamp_Scum_Mask
   samplerSource=SSM_Wrap_WorldGroupSettings
   samplerType=SAMPLERTYPE_Masks
   texture=/Clyde/ExtraDependencies/Swamp_Mask
  }
  n10 = LinearInterpolate(A=@n4, B=Multiply(A=@n5.RGB, B=@n5.RGB), Alpha=Saturate(None=LinearInterpolate(A=@n6, B=@n7, Alpha=@n9.R)))
  n11 = FunctionInput() {bUsePreviewValueAsDefault=true, inputName=Alpha Mask, inputType=FunctionInput_Scalar, sortPriority=4, previewValue=<1 0 0 1>}
  n12 = LinearInterpolate(
   A=ScalarParameter() {defaultValue=1.2575, group=Foam and Pondscum, parameterName=Depth Pondscum Offset_1}
   B=ScalarParameter() {defaultValue=0.5948, group=Foam and Pondscum, parameterName=Depth Pondscum Offset_2}
   Alpha=@n9.R
  )
  n13 = Multiply(A=FunctionInput() {bUsePreviewValueAsDefault=true, inputName=Depth, inputType=FunctionInput_Scalar, sortPriority=5}) {constB=-0.01}
  n14 = Add(
   A=LinearInterpolate(
    A=ScalarParameter() {defaultValue=0.9905, group=Foam and Pondscum, parameterName=Pondscum Texture Offset_1}
    B=ScalarParameter() {defaultValue=-1.5, group=Foam and Pondscum, parameterName=Pondscum Texture Offset_2}
    Alpha=@n9.B
   )
   B=Add(A=@n9.G, B=Multiply(A=@n12, B=@n13))
  )
  n15 = FunctionInput() {bUsePreviewValueAsDefault=true, inputName=Sim Erasure, inputType=FunctionInput_Scalar, sortPriority=2}
  n16 = Max(A=LinearInterpolate(B=Subtract(A=@n14, B=@n15), Alpha=ScalarParameter() {defaultValue=1.0761, group=Foam and Pondscum, parameterName=Pondscum Contrast}) {constA=1}) {constB=0}
  n17 = Multiply(A=@n16, B=ScalarParameter() {defaultValue=1, group=Foam and Pondscum, parameterName=Pondscum Mask}) {constB=10}
  n18 = FunctionOutput(None=Multiply(A=Saturate(None=@n17), B=@n11)) {outputName=Mask, sortPriority=1}
  n19 = TextureSample(UVs=Divide(A=@n8, B=ScalarParameter() {defaultValue=900, group=Foam and Pondscum, parameterName=Pondscum Detail normal Tiling})) {
   samplerSource=SSM_Wrap_WorldGroupSettings
   samplerType=SAMPLERTYPE_Normal
   texture=/Game/Global/Textures/Default/Tileable_Patterns/T_TRV_Moss_01_N
  }
  n20 = SetMaterialAttributes!(
   MaterialAttributes=@n2.MaterialAttributes
   BaseColor=LinearInterpolate(A=@n2.BaseColor, B=FunctionOutput(None=@n10) {outputName=Color, sortPriority=1}, Alpha=@n11)
   Specular=LinearInterpolate(A=@n2.Specular, B=ScalarParameter() {defaultValue=0.1, group=Foam and Pondscum, parameterName=Pondscum Specular}, Alpha=@n18)
   Roughness=LinearInterpolate(A=@n2.Roughness, B=ScalarParameter() {defaultValue=0.3, group=Foam and Pondscum, parameterName=Pondscum Roughness}, Alpha=@n18)
   Opacity=LinearInterpolate(A=@n2.Opacity, B=ScalarParameter() {defaultValue=2, group=Foam and Pondscum, parameterName=Pondscum Opacity}, Alpha=@n18)
   Normal=LinearInterpolate(A=@n2.Normal, B=FunctionOutput(None=@n19.RGB) {outputName=Normal, sortPriority=3}, Alpha=@n18)
  ) {attributeSetTypes [
    #2
    #3
    #4
    #5
    #6
   ]}
  n21 = SetMaterialAttributes!(MaterialAttributes=@n2.MaterialAttributes, BaseColor=LinearInterpolate(A=@n2.BaseColor, B=@n4, Alpha=@n11)) {attributeSetTypes [#2]}
  FunctionOutput(None=ShadingPathSwitch(Default=@n20, Deferred=@n20, Forward=@n21, Mobile=@n21)) {outputName=Attributes}
  Multiply(
   A=ComponentMask(None=@n1.WorldPositionOffset) {b=true, desc=WPO}
   B=ScalarParameter() {defaultValue=-0.02, group=Foam and Pondscum, parameterName="WPO Foam "}
  )
  Multiply(A=@n10, B=@n16)
 ]}
