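# Material function that outputs a depth-attenuated wave WPO and normal, apparently to make particles follow the water surface.
# Position, water body index and depth inputs default to world position and dynamic parameters, though the wave function call is left unspecified.
# Depth relative to MPC_Water's TargetWaveDepth drives a 1-exp(-x) falloff that scales displacement and flattens normals in shallows.
# A static switch, on by default, drops XY offsets to keep only vertical motion.
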MaterialFunction /Clyde/Materials/Functions/WaterParticleDisplacement
graph {nodes [
  FunctionInput(Preview=WorldPosition() {worldPositionShaderOffset=WPT_ExcludeAllShaderOffsets}.XYZ) {bUsePreviewValueAsDefault=true, inputName=Position}
  n1 = DynamicParameter() {parameterIndex=3, defaultValue=<rgba 0 512 0 0>, paramNames [WaterBodyIndex, WaterDepth, Param3, Param4]}
  FunctionInput(Preview=@n1.WaterBodyIndex) {bUsePreviewValueAsDefault=true, inputName=WaterBodyIndex, inputType=FunctionInput_Scalar, sortPriority=1}
  n2 = Constant2Vector()
  n3 = MaterialFunctionCall()
  MaterialExpressionDivide_5 = Divide(
   A=FunctionInput(Preview=@n1.WaterDepth) {bUsePreviewValueAsDefault=true, inputName=Depth, inputType=FunctionInput_Scalar, sortPriority=2}
   B=Divide(A=CollectionParameter() {collection=/Water/Materials/MPC/MPC_Water, parameterName=TargetWaveDepth})
  )
  n4 = Custom!(x=@MaterialExpressionDivide_5) {
   code="return 1-exp(-x);"
   description=exp
   outputType=CMOT_Float1
   inputs [{inputName=x, input {expression=&self:MaterialExpressionDivide_5}}]
  }
  n5 = Multiply(A=@n3, B=@n4)
  n6 = ComponentMask(None=@n5) {b=true}
  n7 = StaticBool() {value=true}
  FunctionOutput(None=StaticSwitch(
    True=AppendVector(A=@n2, B=@n6)
    False=@n5
    Value=FunctionInput(Preview=@n7) {bUsePreviewValueAsDefault=true, inputName=Ignore XY Offsets, inputType=FunctionInput_StaticBool}
   )) {outputName=WPO, sortPriority=1}
  FunctionOutput(None=MaterialFunctionCall(Normal=@n3, Flatness=OneMinus(None=@n4)) {materialFunction=&FlattenNormal}.Result) {outputName=Normal}
  StaticBool() {desc=This method is disabled because it doesn't account for steepness}
 ]}
