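# Material function for river water that adds flowmap wave normals, height displacement, foam and scattering to incoming material attributes.
# A vectorized 1-exp(-x) of depth attenuates velocity and waves, whose source function call shows no assigned function.
# Velocity, depth, sim erasure and low-mip tiled noise build the foam mask, which blends specular, roughness and opacity and lifts WPO.
# Low/Medium quality and Forward/Mobile paths skip those foam blends and lift.
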MaterialFunction /Clyde/Materials/Functions/WaterRiverFlowmaps
graph {nodes [
  AppendVector()
  ComponentMask() {r=true}
  Constant()
  n1 = ScalarParameter() {defaultValue=1250, group=Flowmaps, parameterName=FlowmapWaveNormalScale}
  n2 = FunctionInput(Preview=Constant()) {bUsePreviewValueAsDefault=true, inputName=Far Mask Override, inputType=FunctionInput_Scalar, sortPriority=3}
  Divide(
   A=ComponentMask(None=WorldPosition() {worldPositionShaderOffset=WPT_ExcludeAllShaderOffsets}.XYZ) {g=true, r=true}
   B=LinearInterpolate(A=@n1, B=ScalarParameter() {defaultValue=8000, group=Distance Blending, parameterName=FarScale}, Alpha=@n2)
  )
  n3 = MaterialFunctionCall()
  n4 = ConstantBiasScale(None=@n3) {bias=-0.5, scale=2}
  n5 = FunctionInput() {bUsePreviewValueAsDefault=true, inputName=Depth, inputType=FunctionInput_Scalar, sortPriority=4, previewValue=<1 0 0 1>}
  n6 = ScalarParameter() {defaultValue=256, group=Flowmaps, parameterName=Target Depth for Max Velocity Flow}
  MaterialExpressionDivide_5 = Divide(A=@n5, B=AppendVector(A=Divide(A=@n6), B=ScalarParameter() {defaultValue=128, group=Flowmaps, parameterName=Target Depth for Flowmap Waves}))
  n7 = Custom!(x=@MaterialExpressionDivide_5) {
   code="return 1-exp(-x);"
   description=exp
   outputType=CMOT_Float2
   inputs [{inputName=x, input {expression=&self:MaterialExpressionDivide_5}}]
  }
  n8 = Multiply(A=ComponentMask(None=@n4) {b=true}, B=ComponentMask(None=@n7) {g=true})
  n9 = ScalarParameter() {group=Flowmaps, parameterName=Wave Z Displacement}
  FunctionOutput(None=AppendVector(A=Constant2Vector(), B=Multiply(A=@n8, B=@n9))) {outputName=Flowmap WPO, sortPriority=1}
  n10 = FunctionInput() {bUsePreviewValueAsDefault=true, inputName=Velocity, sortPriority=2, previewValue=<1 0 0 1>}
  n11 = ComponentMask(None=@n7) {r=true}
  n12 = Multiply(A=@n10, B=@n11)
  n13 = ComponentMask(None=@n12) {b=true}
  n14 = Subtract(
   A=Multiply(A=Divide(A=@n13) {constB=100}, B=ScalarParameter() {defaultValue=0.4, group=Flowmaps, parameterName=Velocity Foam Amount})
   B=Multiply(A=@n5, B=ScalarParameter() {defaultValue=0.002, parameterName=Depth Foam Removal})
  )
  n15 = Subtract(A=@n14, B=FunctionInput() {bUsePreviewValueAsDefault=true, inputName=Sim Foam Erasure, inputType=FunctionInput_Scalar, sortPriority=1})
  n16 = ComponentMask(None=WorldPosition() {worldPositionShaderOffset=WPT_ExcludeAllShaderOffsets}.XYZ) {g=true, r=true}
  n17 = ScalarParameter() {defaultValue=3240, parameterName=Flowmap Breakup Scale}
  n18 = Divide(A=@n16, B=@n17)
  n19 = ScalarParameter() {defaultValue=4.5, parameterName=Foam Phase Mip Scatter}
  n20 = Subtract(
   A=Add(
    A=@n15
    B=TextureSample!(UVs=@n18, Tex, Level=@n19, Apply View MipBias) {mipValueMode=TMVM_MipLevel, samplerSource=SSM_Wrap_WorldGroupSettings, texture=/FortWater/Textures/T_WaterFlow_01_Foam_Tiled}.R
   )
   B=ScalarParameter() {defaultValue=1.25, parameterName=Foam Scattering Offset}
  )
  n21 = VectorParameter() {parameterName=Foam Scattering, defaultValue=<rgba 0.6 1 1 0.4>}
  n22 = Multiply(A=@n21.RGB, B=Divide(A=@n21.A) {constB=100})
  FunctionOutput(None=Saturate(None=Multiply(A=@n20, B=@n22))) {outputName=Scattering, sortPriority=5}
  n23 = GetMaterialAttributes(None=FunctionInput() {inputName=Attributes, inputType=FunctionInput_MaterialAttributes}) {attributeGetTypes [
    #1
    #2
    #3
    #4
    #5
    #6
   ]}
  n24 = VectorParameter() {parameterName=Foam Color 1, defaultValue=<rgba 0.6 0.9 1 0.45>}
  n25 = Multiply(A=@n24.RGB, B=@n24.RGB)
  n26 = Add(
   A=Multiply(A=@n8, B=ScalarParameter() {defaultValue=0.4, parameterName=WPO River Foam})
   B=Multiply(A=@n3, B=ScalarParameter() {defaultValue=0.5, parameterName=Foam Texture Amount})
  )
  n27 = TextureSample(UVs=@n18, Level=ScalarParameter() {defaultValue=2.5, parameterName=Foam Phase Mip Level}) {mipValueMode=TMVM_MipLevel, samplerSource=SSM_Wrap_WorldGroupSettings, texture=&T_WaterFlow_01_Foam_Tiled}
  n28 = Multiply(A=@n27.R, B=ScalarParameter() {defaultValue=1, parameterName=Phase offset Foam Amount})
  n29 = Subtract(A=Add(A=Add(A=@n26, B=@n28), B=@n15), B=ScalarParameter() {defaultValue=1.7, group=Flowmaps, parameterName=River Foam offset})
  n30 = ScalarParameter() {defaultValue=1, parameterName=River Foam Boost}
  n31 = FunctionOutput(None=Saturate(None=Multiply(A=@n29, B=@n30))) {outputName=Foam Mask, sortPriority=10}
  n32 = Divide(A=Max(A=ComponentMask(None=@n10) {b=true}) {constB=0.001}, B=ScalarParameter() {defaultValue=512, parameterName=Velocity for Max Normals})
  n33 = Multiply(
   A=Multiply(A=ComponentMask(None=@n4) {g=true, r=true}, B=@n11)
   B=Multiply(A=ScalarParameter() {defaultValue=0.3, group=Flowmaps, parameterName=WavelNormalIntensity}, B=Saturate(None=@n32))
  )
  n34 = Add(
   A=Multiply(A=@n23.Normal, B=Constant3Vector() {constant=<rgba 1 1 0 0>})
   B=FunctionOutput(None=DeriveNormalZ(InXY=@n33)) {outputName=Flowmap Normals}
  )
  n35 = Multiply(A=Max(A=@n29) {constB=0}, B=ScalarParameter() {defaultValue=16, group=Foam and Pondscum, parameterName=Foam WPO})
  n36 = Constant3Vector() {constant=<rgba 0 0 1 0>}
  n37 = SetMaterialAttributes!(
   MaterialAttributes=@n23.MaterialAttributes
   BaseColor=@n25
   Specular=LinearInterpolate(A=@n23.Specular, B=ScalarParameter() {defaultValue=0.1, group=Foam and Pondscum, parameterName=River Foam Specular}, Alpha=@n31)
   Roughness=LinearInterpolate(A=@n23.Roughness, B=ScalarParameter() {defaultValue=2, group=Foam and Pondscum, parameterName=River Foam Roughness}, Alpha=@n31)
   Opacity=LinearInterpolate(A=@n23.Opacity, B=ScalarParameter() {defaultValue=0.6, group=Foam and Pondscum, parameterName=River Foam Opacity}, Alpha=@n31)
   Normal=@n34
   WorldPositionOffset=Add(A=@n23.WorldPositionOffset, B=Multiply(A=@n35, B=@n36))
  ) {attributeSetTypes [
    #1
    #2
    #3
    #4
    #5
    #6
   ]}
  n38 = SetMaterialAttributes(
   MaterialAttributes=@n23.MaterialAttributes
   BaseColor=@n25
   Specular=@n23.MaterialAttributes
   Roughness=@n23.MaterialAttributes
   Opacity=@n23.MaterialAttributes
   Normal=@n34
   WorldPositionOffset=@n23.MaterialAttributes
  ) {attributeSetTypes [
    #1
    #2
    #3
    #4
    #5
    #6
   ]}
  n39 = QualitySwitch(Default=@n37, Low=@n38, High=@n37, Medium=@n38)
  FunctionOutput(None=ShadingPathSwitch(Default=@n39, Deferred=@n39, Forward=@n38, Mobile=@n38)) {outputName=Attributes}
  FunctionOutput(None=@n25) {outputName=Foam, sortPriority=10}
  FunctionOutput(None=AppendVector(A=@n27.R, B=ComponentMask(None=@n3) {r=true})) {
   description="Outputs 4 exponential masks using a single vectorized Exp operation. R=Wave Masked based on Depth, G=Flowmap Mask based on Depth, B=Coloration Mask based on Depth, A=Flow Mask based on velocity"
   outputName=Phase Offset And Foam
   sortPriority=10
  }
  LinearInterpolate()
  n40 = ScalarParameter() {defaultValue=1.5, parameterName=Final Flow Mult}
  n41 = Divide(A=ComponentMask(None=@n12) {g=true, r=true}, B=@n1)
  n42 = ScalarParameter() {defaultValue=0.75, group=Flowmaps, parameterName=Phase Fade Time}
  LinearInterpolate(A=Divide(A=Multiply(A=@n40, B=@n41), B=@n42), Alpha=@n2) {constB=0}
  LinearInterpolate(A=@n34, B=Constant3Vector() {constant=<rgba 0 0 1 0>}, Alpha=@n31)
  n43 = Subtract(A=@n5, B=ScalarParameter() {parameterName=Depth Foam Mask Offset})
  n44 = ScalarParameter() {defaultValue=32, parameterName=DepthMaxFoam}
  n45 = OneMinus(None=Saturate(None=Divide(A=@n43, B=@n44)))
  n46 = ScalarParameter() {parameterName=fertqwehgf}
  Reroute(None=Add(A=@n14, B=Multiply(A=@n45, B=@n46)))
  n47 = MaterialFunctionCall(Speed=@n42) {materialFunction=&TimeWithSpeedVariable}
  n48 = Add(A=@n47.Time, B=Multiply(A=@n27.R, B=ScalarParameter() {defaultValue=-0.3, group=Flowmaps, parameterName=Water Flowmap offset Amount})) {constB=0}
  n49 = QualitySwitch(Default=@n48, Low=@n47.Time, High=@n48, Medium=@n47.Time)
  Reroute(None=ShadingPathSwitch(Default=@n49, Deferred=@n49, Forward=@n47.Time, Mobile=@n47.Time))
  Reroute(None=@n31)
  TextureObjectParameter() {group=Flowmaps, parameterName=FlowmapWaveNormalHeight, samplerType=SAMPLERTYPE_LinearColor, texture=null}
  n50 = Multiply(A=ComponentMask(None=@n10) {g=true, r=true}, B=ScalarParameter() {parameterName=Foam Distortion Advection})
  TextureSample(UVs=Divide(A=Subtract(A=@n16, B=@n50), B=@n17), Level=ScalarParameter() {defaultValue=2.5, parameterName=Foam Alpha Mip Level}) {mipValueMode=TMVM_MipLevel, samplerSource=SSM_Wrap_WorldGroupSettings, texture=&T_WaterFlow_01_Foam_Tiled}
  VectorParameter() {group=Flowmaps, parameterName=Flowmap UV2 Offset and Scale, defaultValue=<rgba 0.5 0.5 1 0>}
 ]}
